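# Widget Blueprint for a contextual tutorial indicator, an animated marker built from icon, sign, diamond box and small arrow images.
# Separate animations cover spin-in entrances, idle, move reaction, disappearance, and box and arrow fades.
# Entrances and reactions use keyed squash-and-stretch scale, rotation and opacity on the overlays and images.
# Most animations also set child visibility, so each state shows only its own layers, including separate solo and child copies of the box.
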WidgetBlueprint /Game/Athena/Tutorial/Contextual/ContextTutorialIndicator
animations [
 ST_in {
  binding_count=13
  length_seconds=0.5956
  object=&self:ST_in
  sequence {bindings [
    Icon {
     id=#1
     possessed_class=Image
     tracks [
      {
       type=MovieSceneColorTrack
       display_name=Color and Opacity
       sections [{
         type=MovieSceneColorSection
         end_frame=4
         start_frame=0
         channels [
          Color.R {type=MovieSceneScriptingFloatChannel}
          Color.G {type=MovieSceneScriptingFloatChannel}
          Color.B {type=MovieSceneScriptingFloatChannel}
          Color.A {type=MovieSceneScriptingFloatChannel, keys [[1.6527,0],[4.7083,1]]}
         ]
        }]
      }
      {
       type=MovieScene2DTransformTrack
       sections [{
         type=MovieScene2DTransformSection
         end_frame=9
         start_frame=0
         channels [
          Translation.X {type=MovieSceneScriptingFloatChannel, const [0,1]}
          Translation.Y {type=MovieSceneScriptingFloatChannel, const [0,1]}
          Scale.X {type=MovieSceneScriptingFloatChannel, keys [[2.025,1.7],[4.2697,0.8],[5.7663,1.1],[7.637,0.95],[9.508,1]]}
          Scale.Y {type=MovieSceneScriptingFloatChannel, keys [[2.025,0],[4.2697,1.2],[5.7663,0.9],[7.637,1.05],[9.508,1]]}
         ]
        }]
      }
      {
       type=MovieSceneFloatTrack
       display_name=Render Opacity
       sections [{type=MovieSceneFloatSection, end_frame=0, start_frame=0, channels [None {type=MovieSceneScriptingFloatChannel, const [1,1]}]}]
      }
      {
       type=MovieSceneByteTrack
       display_name=Visibility
       sections [{type=MovieSceneByteSection, end_frame=1, start_frame=0, channels [None {type=MovieSceneScriptingByteChannel, const [4,1]}]}]
      }
     ]
    }
    Overlay_0 {
     id=#2
     possessed_class=Overlay
     tracks [{
       type=MovieSceneByteTrack
       display_name=Visibility
       sections [{type=MovieSceneByteSection, end_frame=0, start_frame=0, channels [None {type=MovieSceneScriptingByteChannel, const [4,1]}]}]
      }]
    }
    Circle {
     id=#3
     possessed_class=Image
     tracks [{
       type=MovieSceneByteTrack
       display_name=Visibility
       sections [{type=MovieSceneByteSection, end_frame=2, start_frame=0, channels [None {type=MovieSceneScriptingByteChannel, const [1,1]}]}]
      }]
    }
    DescriptionOLD {
     id=#4
     possessed_class=EditableTextBox
     tracks [
      {
       type=MovieSceneByteTrack
       display_name=Visibility
       sections [{type=MovieSceneByteSection, end_frame=11, start_frame=0, channels [None {type=MovieSceneScriptingByteChannel, const [1,1]}]}]
      }
      {
       type=MovieSceneFloatTrack
       display_name=Render Opacity
       sections [{type=MovieSceneFloatSection, end_frame=0, start_frame=0, channels [None {type=MovieSceneScriptingFloatChannel, const [1,1]}]}]
      }
     ]
    }
    Image_SmallIndicator_Child {
     id=#5
     possessed_class=Image
     tracks [{
       type=MovieSceneByteTrack
       display_name=Visibility
       sections [{type=MovieSceneByteSection, end_frame=11, start_frame=0, channels [None {type=MovieSceneScriptingByteChannel, const [1,2]}]}]
      }]
    }
    Switcher_IndicatorSize {
     id=#6
     possessed_class=WidgetSwitcher
     tracks [{
       type=MovieSceneByteTrack
       display_name=Visibility
       sections [{type=MovieSceneByteSection, end_frame=0, start_frame=0, channels [None {type=MovieSceneScriptingByteChannel, const [4,1]}]}]
      }]
    }
    White_Box_Child {
     id=#7
     possessed_class=Image
     tracks [{
       type=MovieSceneByteTrack
       display_name=Visibility
       sections [{type=MovieSceneByteSection, end_frame=11, start_frame=0, channels [None {type=MovieSceneScriptingByteChannel, const [2,2]}]}]
      }]
    }
    SizeBox_1 {
     id=#8
     possessed_class=SizeBox
     tracks [{
       type=MovieSceneByteTrack
       display_name=Visibility
       sections [{type=MovieSceneByteSection, end_frame=0, start_frame=0, channels [None {type=MovieSceneScriptingByteChannel, const [4,1]}]}]
      }]
    }
    Sign {
     id=#9
     possessed_class=Image
     tracks [
      {
       type=MovieSceneColorTrack
       display_name=Color and Opacity
       sections [{
         type=MovieSceneColorSection
         end_frame=7
         start_frame=0
         channels [
          Color.R {type=MovieSceneScriptingFloatChannel}
          Color.G {type=MovieSceneScriptingFloatChannel}
          Color.B {type=MovieSceneScriptingFloatChannel}
          Color.A {type=MovieSceneScriptingFloatChannel, keys [[3.6253,0],[7.1077,1]]}
         ]
        }]
      }
      {
       type=MovieScene2DTransformTrack
       sections [{
         type=MovieScene2DTransformSection
         end_frame=11
         start_frame=0
         channels [
          Translation.X {type=MovieSceneScriptingFloatChannel, const [0,1]}
          Translation.Y {type=MovieSceneScriptingFloatChannel, const [0,1]}
          Scale.X {type=MovieSceneScriptingFloatChannel, keys [[4.0767,1.7],[6.4257,0.8],[7.9917,1.1],[9.95,0.95],[11.9077,1]]}
          Scale.Y {type=MovieSceneScriptingFloatChannel, keys [[4.0767,0],[6.4257,1.2],[7.9917,0.9],[9.95,1.05],[11.9077,1]]}
         ]
        }]
      }
      {
       type=MovieSceneFloatTrack
       display_name=Render Opacity
       sections [{type=MovieSceneFloatSection, end_frame=0, start_frame=0, channels [None {type=MovieSceneScriptingFloatChannel, const [1,1]}]}]
      }
      {
       type=MovieSceneByteTrack
       display_name=Visibility
       sections [{type=MovieSceneByteSection, end_frame=0, start_frame=0, channels [None {type=MovieSceneScriptingByteChannel, const [4,1]}]}]
      }
     ]
    }
    Overlay_Icon_Box {
     id=#10
     possessed_class=Overlay
     tracks [{
       type=MovieSceneByteTrack
       display_name=Visibility
       sections [{type=MovieSceneByteSection, end_frame=0, start_frame=0, channels [None {type=MovieSceneScriptingByteChannel, const [4,1]}]}]
      }]
    }
    Image_SmallIndicator {
     id=#11
     possessed_class=Image
     tracks [
      {
       type=MovieSceneColorTrack
       display_name=Color and Opacity
       sections [{
         type=MovieSceneColorSection
         end_frame=2
         start_frame=0
         channels [
          Color.R {type=MovieSceneScriptingFloatChannel}
          Color.G {type=MovieSceneScriptingFloatChannel}
          Color.B {type=MovieSceneScriptingFloatChannel}
          Color.A {type=MovieSceneScriptingFloatChannel, keys [[0,0],[2.1093,1]]}
         ]
        }]
      }
      {
       type=MovieScene2DTransformTrack
       sections [{
         type=MovieScene2DTransformSection
         end_frame=7
         start_frame=0
         channels [
          Translation.X {type=MovieSceneScriptingFloatChannel, const [0,1]}
          Translation.Y {type=MovieSceneScriptingFloatChannel, const [46,1]}
          Scale.X {type=MovieSceneScriptingFloatChannel, keys [[0,1.7],[2.0873,0.8],[3.4807,1.1],[5.222,0.95],[6.9637,1]]}
          Scale.Y {type=MovieSceneScriptingFloatChannel, keys [[0,0],[2.0873,1.2],[3.4807,0.9],[5.222,1.05],[6.9637,1]]}
         ]
        }]
      }
      {
       type=MovieSceneByteTrack
       display_name=Visibility
       sections [{type=MovieSceneByteSection, end_frame=0, start_frame=0, channels [None {type=MovieSceneScriptingByteChannel, const [4,1]}]}]
      }
      {
       type=MovieSceneFloatTrack
       display_name=Render Opacity
       sections [{type=MovieSceneFloatSection, end_frame=0, start_frame=0, channels [None {type=MovieSceneScriptingFloatChannel, const [1,1]}]}]
      }
     ]
    }
    White_Box_Solo {
     id=#12
     possessed_class=Image
     tracks [{
       type=MovieSceneByteTrack
       display_name=Visibility
       sections [{type=MovieSceneByteSection, end_frame=11, start_frame=0, channels [None {type=MovieSceneScriptingByteChannel, const [1,1]}]}]
      }]
    }
    Overlay_Icon {
     id=#13
     possessed_class=Overlay
     tracks [
      {
       type=MovieSceneByteTrack
       display_name=Visibility
       sections [{type=MovieSceneByteSection, end_frame=11, start_frame=0, channels [None {type=MovieSceneScriptingByteChannel, const [4,1]}]}]
      }
      {
       type=MovieSceneFloatTrack
       display_name=Render Opacity
       sections [{type=MovieSceneFloatSection, end_frame=0, start_frame=0, channels [None {type=MovieSceneScriptingFloatChannel, const [1,1]}]}]
      }
      {
       type=MovieScene2DTransformTrack
       sections [{
         type=MovieScene2DTransformSection
         end_frame=0
         start_frame=0
         channels [
          Translation.X {type=MovieSceneScriptingFloatChannel, const [0,1]}
          Translation.Y {type=MovieSceneScriptingFloatChannel, const [0,1]}
          Angle {type=MovieSceneScriptingFloatChannel, const [0,1]}
          Scale.X {type=MovieSceneScriptingFloatChannel, const [1,1]}
          Scale.Y {type=MovieSceneScriptingFloatChannel, const [1,1]}
          Shear.X {type=MovieSceneScriptingFloatChannel, const [0,1]}
          Shear.Y {type=MovieSceneScriptingFloatChannel, const [0,1]}
         ]
        }]
      }
     ]
    }
   ]}
 }
 ST_border {
  binding_count=7
  length_seconds=0.45
  object=&self:ST_border
  sequence {bindings [
    Image_SmallIndicator_Child {
     id=#14
     possessed_class=Image
     tracks [{
       type=MovieSceneByteTrack
       display_name=Visibility
       sections [{type=MovieSceneByteSection, end_frame=8, start_frame=0, channels [None {type=MovieSceneScriptingByteChannel, const [1,1]}]}]
      }]
    }
    Image_SmallIndicator {
     id=#15
     possessed_class=Image
     tracks [
      {
       type=MovieScene2DTransformTrack
       sections [{
         type=MovieScene2DTransformSection
         end_frame=8
         start_frame=0
         channels [
          Scale.X {type=MovieSceneScriptingFloatChannel, keys [[0,1],[2.0837,0.6]]}
          Scale.Y {type=MovieSceneScriptingFloatChannel, keys [[0,1],[2.0837,1.4]]}
         ]
        }]
      }
      {
       type=MovieSceneColorTrack
       display_name=Color and Opacity
       sections [{
         type=MovieSceneColorSection
         end_frame=8
         start_frame=0
         channels [
          Color.R {type=MovieSceneScriptingFloatChannel}
          Color.G {type=MovieSceneScriptingFloatChannel}
          Color.B {type=MovieSceneScriptingFloatChannel}
          Color.A {type=MovieSceneScriptingFloatChannel, keys [[0,1],[2.0837,0]]}
         ]
        }]
      }
      {type=MovieSceneByteTrack, display_name=Visibility, sections []}
     ]
    }
    Circle {
     id=#16
     possessed_class=Image
     tracks [{
       type=MovieSceneByteTrack
       display_name=Visibility
       sections [{type=MovieSceneByteSection, end_frame=8, start_frame=0, channels [None {type=MovieSceneScriptingByteChannel, const [1,1]}]}]
      }]
    }
    White_Box_Solo {
     id=#17
     possessed_class=Image
     tracks [{
       type=MovieSceneByteTrack
       display_name=Visibility
       sections [{type=MovieSceneByteSection, end_frame=9, start_frame=0, channels [None {type=MovieSceneScriptingByteChannel, const [1,1]}]}]
      }]
    }
    DescriptionOLD {
     id=#18
     possessed_class=EditableTextBox
     tracks [{
       type=MovieSceneByteTrack
       display_name=Visibility
       sections [{type=MovieSceneByteSection, end_frame=8, start_frame=0, channels [None {type=MovieSceneScriptingByteChannel, const [1,1]}]}]
      }]
    }
    White_Box_Child {
     id=#19
     possessed_class=Image
     tracks [{
       type=MovieSceneByteTrack
       display_name=Visibility
       sections [{type=MovieSceneByteSection, end_frame=9, start_frame=0, channels [None {type=MovieSceneScriptingByteChannel, const [2,1]}]}]
      }]
    }
    Overlay_Icon {
     id=#20
     possessed_class=Overlay
     tracks [{
       type=MovieScene2DTransformTrack
       sections [{
         type=MovieScene2DTransformSection
         end_frame=9
         start_frame=0
         channels [
          Translation.X {type=MovieSceneScriptingFloatChannel, const [0,1]}
          Translation.Y {type=MovieSceneScriptingFloatChannel, const [0,1]}
          Scale.X {type=MovieSceneScriptingFloatChannel, keys [[0,1],[2.0837,0.6],[4.42,1.2],[6.567,0.98],[8.651,1.01],[10.7977,1]]}
          Scale.Y {type=MovieSceneScriptingFloatChannel, keys [[0,1],[2.0837,1.4],[4.42,0.8],[6.567,1.02],[8.651,0.99],[10.7977,1]]}
         ]
        }]
      }]
    }
   ]}
 }
 ST_in_Diamond_Spin {
  binding_count=13
  length_seconds=0.6
  object=&self:ST_in_Diamond_Spin
  sequence {bindings [
    White_Box_Child {
     id=#21
     possessed_class=Image
     tracks [
      {
       type=MovieSceneFloatTrack
       display_name=Render Opacity
       sections [{
         type=MovieSceneFloatSection
         end_frame=12
         start_frame=0
         channels [None {type=MovieSceneScriptingFloatChannel, keys [[0,0],[2.1427,1],[9.9997,1]]}]
        }]
      }
      {
       type=MovieSceneByteTrack
       display_name=Visibility
       sections [{type=MovieSceneByteSection, end_frame=12, start_frame=0, channels [None {type=MovieSceneScriptingByteChannel, const [2,3]}]}]
      }
     ]
    }
    White_Box_Solo {
     id=#22
     possessed_class=Image
     tracks [{
       type=MovieSceneByteTrack
       display_name=Visibility
       sections [{type=MovieSceneByteSection, end_frame=11, start_frame=0, channels [None {type=MovieSceneScriptingByteChannel, const [1,1]}]}]
      }]
    }
    Icon {
     id=#23
     possessed_class=Image
     tracks [{
       type=MovieSceneFloatTrack
       display_name=Render Opacity
       sections [{
         type=MovieSceneFloatSection
         end_frame=12
         start_frame=0
         channels [None {type=MovieSceneScriptingFloatChannel, keys [[0,0],[0.714,0],[4.2853,1],[9.9997,1]]}]
        }]
      }]
    }
    Switcher_IndicatorSize {
     id=#24
     possessed_class=WidgetSwitcher
     tracks [{
       type=MovieSceneByteTrack
       display_name=Visibility
       sections [{type=MovieSceneByteSection, end_frame=0, start_frame=0, channels [None {type=MovieSceneScriptingByteChannel, const [4,1]}]}]
      }]
    }
    SizeBox_1 {
     id=#25
     possessed_class=SizeBox
     tracks [{
       type=MovieSceneByteTrack
       display_name=Visibility
       sections [{type=MovieSceneByteSection, end_frame=0, start_frame=0, channels [None {type=MovieSceneScriptingByteChannel, const [4,1]}]}]
      }]
    }
    Image_SmallIndicator {
     id=#26
     possessed_class=Image
     tracks [{
       type=MovieSceneByteTrack
       display_name=Visibility
       sections [{type=MovieSceneByteSection, end_frame=12, start_frame=0, channels [None {type=MovieSceneScriptingByteChannel, const [2,2]}]}]
      }]
    }
    DescriptionOLD {
     id=#27
     possessed_class=EditableTextBox
     tracks [{
       type=MovieSceneByteTrack
       display_name=Visibility
       sections [{type=MovieSceneByteSection, end_frame=11, start_frame=0, channels [None {type=MovieSceneScriptingByteChannel, const [1,1]}]}]
      }]
    }
    Overlay_Icon_Box {
     id=#28
     possessed_class=Overlay
     tracks [
      {
       type=MovieScene2DTransformTrack
       sections [{
         type=MovieScene2DTransformSection
         end_frame=12
         start_frame=0
         channels [
          Translation.X {type=MovieSceneScriptingFloatChannel}
          Translation.Y {type=MovieSceneScriptingFloatChannel}
          Angle {type=MovieSceneScriptingFloatChannel, keys [[0,180],[4.2853,0],[9.9997,0]]}
          Scale.X {type=MovieSceneScriptingFloatChannel, keys [[0,3.5],[4.2853,1],[9.9997,1]]}
          Scale.Y {type=MovieSceneScriptingFloatChannel, keys [[0,3.5],[4.2853,1],[9.9997,1]]}
          Shear.X {type=MovieSceneScriptingFloatChannel}
          Shear.Y {type=MovieSceneScriptingFloatChannel}
         ]
        }]
      }
      {
       type=MovieSceneByteTrack
       display_name=Visibility
       sections [{type=MovieSceneByteSection, end_frame=0, start_frame=0, channels [None {type=MovieSceneScriptingByteChannel, const [4,1]}]}]
      }
     ]
    }
    Sign {
     id=#29
     possessed_class=Image
     tracks [
      {
       type=MovieSceneFloatTrack
       display_name=Render Opacity
       sections [{
         type=MovieSceneFloatSection
         end_frame=12
         start_frame=0
         channels [None {type=MovieSceneScriptingFloatChannel, keys [[0,0],[2.857,0],[4.9997,1],[9.9997,1]]}]
        }]
      }
      {
       type=MovieScene2DTransformTrack
       sections [{
         type=MovieScene2DTransformSection
         end_frame=12
         start_frame=0
         channels [
          Translation.X {type=MovieSceneScriptingFloatChannel}
          Translation.Y {type=MovieSceneScriptingFloatChannel}
          Angle {type=MovieSceneScriptingFloatChannel}
          Scale.X {type=MovieSceneScriptingFloatChannel, keys [[0,1],[3.4963,1],[5.1687,0.4262],[6.4843,1],[7.3123,1.2],[7.7267,1],[9.9997,1]]}
          Scale.Y {type=MovieSceneScriptingFloatChannel, keys [[0,1],[3.4963,1],[5.1687,1.4595],[6.4843,1],[7.3123,0.8],[7.7267,1],[9.9997,1]]}
          Shear.X {type=MovieSceneScriptingFloatChannel}
          Shear.Y {type=MovieSceneScriptingFloatChannel}
         ]
        }]
      }
     ]
    }
    Circle {
     id=#30
     possessed_class=Image
     tracks [{
       type=MovieSceneByteTrack
       display_name=Visibility
       sections [{type=MovieSceneByteSection, end_frame=12, start_frame=0, channels [None {type=MovieSceneScriptingByteChannel, const [2,3]}]}]
      }]
    }
    Overlay_0 {
     id=#31
     possessed_class=Overlay
     tracks [{
       type=MovieSceneByteTrack
       display_name=Visibility
       sections [{type=MovieSceneByteSection, end_frame=0, start_frame=0, channels [None {type=MovieSceneScriptingByteChannel, const [4,1]}]}]
      }]
    }
    Overlay_Box_Solo {
     id=#32
     possessed_class=Overlay
     tracks [{
       type=MovieSceneByteTrack
       display_name=Visibility
       sections [{type=MovieSceneByteSection, end_frame=0, start_frame=0, channels [None {type=MovieSceneScriptingByteChannel, const [4,1]}]}]
      }]
    }
    Image_SmallIndicator_Child {
     id=#33
     possessed_class=Image
     tracks [{
       type=MovieSceneByteTrack
       display_name=Visibility
       sections [{type=MovieSceneByteSection, end_frame=12, start_frame=0, channels [None {type=MovieSceneScriptingByteChannel, const [1,2]}]}]
      }]
    }
   ]}
 }
 ST_Marker_Idle {
  binding_count=6
  length_seconds=0.5
  object=&self:ST_Marker_Idle
  sequence {bindings [
    Image_SmallIndicator {
     id=#34
     possessed_class=Image
     tracks [{
       type=MovieSceneByteTrack
       display_name=Visibility
       sections [{type=MovieSceneByteSection, end_frame=10, start_frame=0, channels [None {type=MovieSceneScriptingByteChannel, const [2,2]}]}]
      }]
    }
    DescriptionOLD {
     id=#35
     possessed_class=EditableTextBox
     tracks [{
       type=MovieSceneByteTrack
       display_name=Visibility
       sections [{type=MovieSceneByteSection, end_frame=10, start_frame=0, channels [None {type=MovieSceneScriptingByteChannel, const [1,1]}]}]
      }]
    }
    Circle {
     id=#36
     possessed_class=Image
     tracks [{
       type=MovieSceneByteTrack
       display_name=Visibility
       sections [{type=MovieSceneByteSection, end_frame=10, start_frame=0, channels [None {type=MovieSceneScriptingByteChannel, const [2,2]}]}]
      }]
    }
    Image_SmallIndicator_Child {
     id=#37
     possessed_class=Image
     tracks [{
       type=MovieSceneByteTrack
       display_name=Visibility
       sections [{type=MovieSceneByteSection, end_frame=0, start_frame=0, channels [None {type=MovieSceneScriptingByteChannel, const [1,1]}]}]
      }]
    }
    White_Box_Child {
     id=#38
     possessed_class=Image
     tracks [{
       type=MovieSceneByteTrack
       display_name=Visibility
       sections [{type=MovieSceneByteSection, end_frame=10, start_frame=0, channels [None {type=MovieSceneScriptingByteChannel, const [2,2]}]}]
      }]
    }
    White_Box_Solo {
     id=#39
     possessed_class=Image
     tracks [{
       type=MovieSceneByteTrack
       display_name=Visibility
       sections [{type=MovieSceneByteSection, end_frame=0, start_frame=0, channels [None {type=MovieSceneScriptingByteChannel, const [1,1]}]}]
      }]
    }
   ]}
 }
 ST_in_DiamondBox_Spin {
  binding_count=9
  length_seconds=0.2102
  object=&self:ST_in_DiamondBox_Spin
  sequence {bindings [
    White_Box_Solo {
     id=#40
     possessed_class=Image
     tracks [{
       type=MovieSceneByteTrack
       display_name=Visibility
       sections [{type=MovieSceneByteSection, end_frame=4, start_frame=0, channels [None {type=MovieSceneScriptingByteChannel, const [1,1]}]}]
      }]
    }
    White_Box_Child {
     id=#21
     possessed_class=Image
     tracks [{
       type=MovieSceneFloatTrack
       display_name=Render Opacity
       sections [{type=MovieSceneFloatSection, end_frame=4, start_frame=0, channels [None {type=MovieSceneScriptingFloatChannel, keys [[0,0],[2.1013,1]]}]}]
      }]
    }
    Icon {
     id=#23
     possessed_class=Image
     tracks [{
       type=MovieSceneByteTrack
       display_name=Visibility
       sections [{type=MovieSceneByteSection, end_frame=4, start_frame=0, channels [None {type=MovieSceneScriptingByteChannel, const [2,1]}]}]
      }]
    }
    Image_SmallIndicator {
     id=#41
     possessed_class=Image
     tracks [{
       type=MovieSceneByteTrack
       display_name=Visibility
       sections [{type=MovieSceneByteSection, end_frame=4, start_frame=0, channels [None {type=MovieSceneScriptingByteChannel, const [2,1]}]}]
      }]
    }
    Overlay_Icon_Box {
     id=#28
     possessed_class=Overlay
     tracks [{
       type=MovieScene2DTransformTrack
       sections [{
         type=MovieScene2DTransformSection
         end_frame=4
         start_frame=0
         channels [
          Translation.X {type=MovieSceneScriptingFloatChannel}
          Translation.Y {type=MovieSceneScriptingFloatChannel}
          Angle {type=MovieSceneScriptingFloatChannel, keys [[0,180],[4.2023,0]]}
          Scale.X {type=MovieSceneScriptingFloatChannel, keys [[0,3.5],[4.2027,1]]}
          Scale.Y {type=MovieSceneScriptingFloatChannel, keys [[0,3.5],[4.2027,1]]}
          Shear.X {type=MovieSceneScriptingFloatChannel}
          Shear.Y {type=MovieSceneScriptingFloatChannel}
         ]
        }]
      }]
    }
    Sign {
     id=#29
     possessed_class=Image
     tracks [{
       type=MovieSceneByteTrack
       display_name=Visibility
       sections [{type=MovieSceneByteSection, end_frame=4, start_frame=0, channels [None {type=MovieSceneScriptingByteChannel, const [2,1]}]}]
      }]
    }
    Circle {
     id=#30
     possessed_class=Image
     tracks [{
       type=MovieSceneByteTrack
       display_name=Visibility
       sections [{type=MovieSceneByteSection, end_frame=4, start_frame=0, channels [None {type=MovieSceneScriptingByteChannel, const [2,1]}]}]
      }]
    }
    Image_SmallIndicator_Child {
     id=#42
     possessed_class=Image
     tracks [{
       type=MovieSceneByteTrack
       display_name=Visibility
       sections [{type=MovieSceneByteSection, end_frame=4, start_frame=0, channels [None {type=MovieSceneScriptingByteChannel, const [1,1]}]}]
      }]
    }
    DescriptionOLD {
     id=#43
     possessed_class=EditableTextBox
     tracks [{
       type=MovieSceneByteTrack
       display_name=Visibility
       sections [{type=MovieSceneByteSection, end_frame=4, start_frame=0, channels [None {type=MovieSceneScriptingByteChannel, const [1,1]}]}]
      }]
    }
   ]}
 }
 ST_Marker_MoveReaction {
  binding_count=8
  length_seconds=0.5
  object=&self:ST_Marker_MoveReaction
  sequence {bindings [
    White_Box_Child {
     id=#21
     possessed_class=Image
     tracks [{
       type=MovieSceneByteTrack
       display_name=Visibility
       sections [{type=MovieSceneByteSection, end_frame=10, start_frame=0, channels [None {type=MovieSceneScriptingByteChannel, const [2,2]}]}]
      }]
    }
    Icon {
     id=#23
     possessed_class=Image
     tracks [
      {
       type=MovieSceneFloatTrack
       display_name=Render Opacity
       sections [{type=MovieSceneFloatSection, end_frame=10, start_frame=0, channels [None {type=MovieSceneScriptingFloatChannel, const [1,1]}]}]
      }
      {
       type=MovieScene2DTransformTrack
       sections [{
         type=MovieScene2DTransformSection
         end_frame=9
         start_frame=0
         channels [
          Scale.X {type=MovieSceneScriptingFloatChannel, keys [[0,1],[1.999,0.6],[3.3983,1.3],[4.2013,1]]}
          Scale.Y {type=MovieSceneScriptingFloatChannel, keys [[0,1],[1.999,1.4],[3.3983,0.7],[4.2013,1]]}
         ]
        }]
      }
     ]
    }
    Image_SmallIndicator_Child {
     id=#44
     possessed_class=Image
     tracks [{
       type=MovieSceneByteTrack
       display_name=Visibility
       sections [{type=MovieSceneByteSection, end_frame=10, start_frame=0, channels [None {type=MovieSceneScriptingByteChannel, const [1,2]}]}]
      }]
    }
    Sign {
     id=#29
     possessed_class=Image
     tracks [{
       type=MovieScene2DTransformTrack
       sections [{
         type=MovieScene2DTransformSection
         end_frame=9
         start_frame=-1
         channels [
          Scale.X {
           type=MovieSceneScriptingFloatChannel
           thinned_from=8
           keys [[-0.0003,1],[2.0967,0.4],[3.3987,1],[4.3447,1.25],[4.3447,1.2],[5.1507,1],[5.9983,1]]
          }
          Scale.Y {type=MovieSceneScriptingFloatChannel, thinned_from=7, keys [[-0.0003,1],[2.0967,1.6],[3.3987,1],[4.3447,0.85],[5.1507,1],[5.9983,1]]}
         ]
        }]
      }]
    }
    White_Box_Solo {
     id=#45
     possessed_class=Image
     tracks [{
       type=MovieSceneByteTrack
       display_name=Visibility
       sections [{type=MovieSceneByteSection, end_frame=9, start_frame=0, channels [None {type=MovieSceneScriptingByteChannel, const [1,1]}]}]
      }]
    }
    Circle {
     id=#30
     possessed_class=Image
     tracks [{
       type=MovieSceneByteTrack
       display_name=Visibility
       sections [{type=MovieSceneByteSection, end_frame=9, start_frame=0, channels [None {type=MovieSceneScriptingByteChannel, const [2,2]}]}]
      }]
    }
    DescriptionOLD {
     id=#46
     possessed_class=EditableTextBox
     tracks [{
       type=MovieSceneByteTrack
       display_name=Visibility
       sections [{type=MovieSceneByteSection, end_frame=9, start_frame=0, channels [None {type=MovieSceneScriptingByteChannel, const [1,1]}]}]
      }]
    }
    Image_SmallIndicator {
     id=#47
     possessed_class=Image
     tracks [{
       type=MovieSceneByteTrack
       display_name=Visibility
       sections [{type=MovieSceneByteSection, end_frame=10, start_frame=0, channels [None {type=MovieSceneScriptingByteChannel, const [2,2]}]}]
      }]
    }
   ]}
 }
 ST_Diamond_Disappear {
  binding_count=8
  length_seconds=0.5
  object=&self:ST_Diamond_Disappear
  sequence {bindings [
    Image_SmallIndicator {
     id=#48
     possessed_class=Image
     tracks [{
       type=MovieSceneByteTrack
       display_name=Visibility
       sections [{type=MovieSceneByteSection, end_frame=9, start_frame=0, channels [None {type=MovieSceneScriptingByteChannel, const [1,1]}]}]
      }]
    }
    White_Box_Child {
     id=#21
     possessed_class=Image
     tracks [{
       type=MovieSceneByteTrack
       display_name=Visibility
       sections [{type=MovieSceneByteSection, end_frame=10, start_frame=0, channels [None {type=MovieSceneScriptingByteChannel, const [2,2]}]}]
      }]
    }
    Icon {
     id=#23
     possessed_class=Image
     tracks [{
       type=MovieSceneFloatTrack
       display_name=Render Opacity
       sections [{
         type=MovieSceneFloatSection
         end_frame=10
         start_frame=0
         channels [None {type=MovieSceneScriptingFloatChannel, keys [[1.0003,1],[3.0003,0],[10.0003,0]]}]
        }]
      }]
    }
    Sign {
     id=#29
     possessed_class=Image
     tracks [{
       type=MovieScene2DTransformTrack
       sections [{
         type=MovieScene2DTransformSection
         end_frame=10
         start_frame=0
         channels [
          Scale.X {type=MovieSceneScriptingFloatChannel, keys [[1.0003,1],[3,0.4],[4.677,1],[10.0003,1]]}
          Scale.Y {type=MovieSceneScriptingFloatChannel, keys [[1.0003,1],[3,1.6],[4.677,1],[5.7833,0.8],[6.493,1],[10.0003,1]]}
         ]
        }]
      }]
    }
    Image_SmallIndicator_Child {
     id=#49
     possessed_class=Image
     tracks [{
       type=MovieSceneByteTrack
       display_name=Visibility
       sections [{type=MovieSceneByteSection, end_frame=10, start_frame=0, channels [None {type=MovieSceneScriptingByteChannel, const [1,2]}]}]
      }]
    }
    Circle {
     id=#30
     possessed_class=Image
     tracks [{
       type=MovieSceneByteTrack
       display_name=Visibility
       sections [{type=MovieSceneByteSection, end_frame=10, start_frame=0, channels [None {type=MovieSceneScriptingByteChannel, const [2,2]}]}]
      }]
    }
    DescriptionOLD {
     id=#50
     possessed_class=EditableTextBox
     tracks [{
       type=MovieSceneByteTrack
       display_name=Visibility
       sections [{type=MovieSceneByteSection, end_frame=10, start_frame=0, channels [None {type=MovieSceneScriptingByteChannel, const [1,1]}]}]
      }]
    }
    White_Box_Solo {
     id=#51
     possessed_class=Image
     tracks [{
       type=MovieSceneByteTrack
       display_name=Visibility
       sections [{type=MovieSceneByteSection, end_frame=10, start_frame=0, channels [None {type=MovieSceneScriptingByteChannel, const [1,1]}]}]
      }]
    }
   ]}
 }
 ST_DiamondBox_Fade {
  binding_count=8
  length_seconds=0.3
  object=&self:ST_DiamondBox_Fade
  sequence {bindings [
    White_Box_Child {
     id=#21
     possessed_class=Image
     tracks [{
       type=MovieSceneByteTrack
       display_name=Visibility
       sections [{type=MovieSceneByteSection, end_frame=6, start_frame=0, channels [None {type=MovieSceneScriptingByteChannel, const [1,1]}]}]
      }]
    }
    Icon {
     id=#23
     possessed_class=Image
     tracks [{
       type=MovieSceneByteTrack
       display_name=Visibility
       sections [{type=MovieSceneByteSection, end_frame=6, start_frame=0, channels [None {type=MovieSceneScriptingByteChannel, const [2,1]}]}]
      }]
    }
    Image_SmallIndicator {
     id=#52
     possessed_class=Image
     tracks [{
       type=MovieSceneByteTrack
       display_name=Visibility
       sections [{type=MovieSceneByteSection, end_frame=6, start_frame=0, channels [None {type=MovieSceneScriptingByteChannel, const [2,1]}]}]
      }]
    }
    Sign {
     id=#29
     possessed_class=Image
     tracks [{
       type=MovieSceneByteTrack
       display_name=Visibility
       sections [{type=MovieSceneByteSection, end_frame=6, start_frame=0, channels [None {type=MovieSceneScriptingByteChannel, const [2,1]}]}]
      }]
    }
    White_Box_Solo {
     id=#53
     possessed_class=Image
     tracks [
      {
       type=MovieSceneFloatTrack
       display_name=Render Opacity
       sections [{type=MovieSceneFloatSection, end_frame=6, start_frame=0, channels [None {type=MovieSceneScriptingFloatChannel, keys [[0,1],[6,0]]}]}]
      }
      {
       type=MovieScene2DTransformTrack
       sections [{
         type=MovieScene2DTransformSection
         end_frame=6
         start_frame=0
         channels [
          Scale.X {type=MovieSceneScriptingFloatChannel, keys [[0,1],[6,2.5]]}
          Scale.Y {type=MovieSceneScriptingFloatChannel, keys [[0,1],[6,2.5]]}
         ]
        }]
      }
     ]
    }
    Circle {
     id=#30
     possessed_class=Image
     tracks [{
       type=MovieSceneByteTrack
       display_name=Visibility
       sections [{type=MovieSceneByteSection, end_frame=6, start_frame=0, channels [None {type=MovieSceneScriptingByteChannel, const [2,1]}]}]
      }]
    }
    Image_SmallIndicator_Child {
     id=#54
     possessed_class=Image
     tracks [{
       type=MovieSceneByteTrack
       display_name=Visibility
       sections [{type=MovieSceneByteSection, end_frame=6, start_frame=0, channels [None {type=MovieSceneScriptingByteChannel, const [1,1]}]}]
      }]
    }
    DescriptionOLD {
     id=#55
     possessed_class=EditableTextBox
     tracks [{
       type=MovieSceneByteTrack
       display_name=Visibility
       sections [{type=MovieSceneByteSection, end_frame=6, start_frame=0, channels [None {type=MovieSceneScriptingByteChannel, const [1,1]}]}]
      }]
    }
   ]}
 }
 ST_in_Spin_Arrow_Marker {
  binding_count=9
  length_seconds=0.45
  object=&self:ST_in_Spin_Arrow_Marker
  sequence {bindings [
    White_Box_Child {
     id=#21
     possessed_class=Image
     tracks [
      {
       type=MovieSceneFloatTrack
       display_name=Render Opacity
       sections [{
         type=MovieSceneFloatSection
         end_frame=8
         start_frame=0
         channels [None {type=MovieSceneScriptingFloatChannel, keys [[0,0],[1.7133,1],[7.9983,1]]}]
        }]
      }
      {
       type=MovieSceneByteTrack
       display_name=Visibility
       sections [{type=MovieSceneByteSection, end_frame=8, start_frame=0, channels [None {type=MovieSceneScriptingByteChannel, const [2,2]}]}]
      }
     ]
    }
    Icon {
     id=#23
     possessed_class=Image
     tracks [{
       type=MovieSceneFloatTrack
       display_name=Render Opacity
       sections [{
         type=MovieSceneFloatSection
         end_frame=8
         start_frame=0
         channels [None {type=MovieSceneScriptingFloatChannel, keys [[0,0],[0.5707,0],[3.4277,1],[7.9987,1]]}]
        }]
      }]
    }
    Overlay_Icon_Box {
     id=#28
     possessed_class=Overlay
     tracks [{
       type=MovieScene2DTransformTrack
       sections [{
         type=MovieScene2DTransformSection
         end_frame=8
         start_frame=0
         channels [
          Translation.X {type=MovieSceneScriptingFloatChannel}
          Translation.Y {type=MovieSceneScriptingFloatChannel}
          Angle {type=MovieSceneScriptingFloatChannel, keys [[0,-180],[3.4277,0],[7.9983,0]]}
          Scale.X {type=MovieSceneScriptingFloatChannel, keys [[0,2],[3.4277,1],[7.9983,1]]}
          Scale.Y {type=MovieSceneScriptingFloatChannel, keys [[0,2],[3.4277,1],[7.9983,1]]}
          Shear.X {type=MovieSceneScriptingFloatChannel}
          Shear.Y {type=MovieSceneScriptingFloatChannel}
         ]
        }]
      }]
    }
    Image_SmallIndicator {
     id=#56
     possessed_class=Image
     tracks [
      {
       type=MovieScene2DTransformTrack
       sections [{
         type=MovieScene2DTransformSection
         end_frame=8
         start_frame=0
         channels [
          Translation.X {type=MovieSceneScriptingFloatChannel, const [0,2]}
          Translation.Y {type=MovieSceneScriptingFloatChannel, const [55,2]}
          Angle {type=MovieSceneScriptingFloatChannel, const [0,2]}
          Scale.X {type=MovieSceneScriptingFloatChannel, const [0.5,2]}
          Scale.Y {type=MovieSceneScriptingFloatChannel, const [0.5,2]}
          Shear.X {type=MovieSceneScriptingFloatChannel, const [0,2]}
          Shear.Y {type=MovieSceneScriptingFloatChannel, const [0,2]}
         ]
        }]
      }
      {
       type=MovieSceneFloatTrack
       display_name=Render Opacity
       sections [{
         type=MovieSceneFloatSection
         end_frame=9
         start_frame=0
         channels [None {type=MovieSceneScriptingFloatChannel, keys [[0,0],[4.57,0],[5.713,1],[7.9987,1]]}]
        }]
      }
     ]
    }
    Image_SmallIndicator_Child {
     id=#57
     possessed_class=Image
     tracks [{
       type=MovieSceneByteTrack
       display_name=Visibility
       sections [{type=MovieSceneByteSection, end_frame=9, start_frame=0, channels [None {type=MovieSceneScriptingByteChannel, const [1,2]}]}]
      }]
    }
    Sign {
     id=#29
     possessed_class=Image
     tracks [
      {
       type=MovieSceneFloatTrack
       display_name=Render Opacity
       sections [{
         type=MovieSceneFloatSection
         end_frame=8
         start_frame=0
         channels [None {type=MovieSceneScriptingFloatChannel, keys [[0,0],[2.2847,0],[3.9987,1],[7.9987,1]]}]
        }]
      }
      {
       type=MovieScene2DTransformTrack
       sections [{
         type=MovieScene2DTransformSection
         end_frame=8
         start_frame=0
         channels [
          Translation.X {type=MovieSceneScriptingFloatChannel}
          Translation.Y {type=MovieSceneScriptingFloatChannel}
          Angle {type=MovieSceneScriptingFloatChannel}
          Scale.X {type=MovieSceneScriptingFloatChannel, keys [[0,1],[2.8563,1],[3.9987,0.4262],[5.1413,1],[6.284,1.2],[7.4263,1],[7.9983,1]]}
          Scale.Y {type=MovieSceneScriptingFloatChannel, keys [[0,1],[2.8563,1],[3.9987,1.4595],[5.1413,1],[6.2837,0.8],[7.4263,1],[7.9983,1]]}
          Shear.X {type=MovieSceneScriptingFloatChannel}
          Shear.Y {type=MovieSceneScriptingFloatChannel}
         ]
        }]
      }
     ]
    }
    DescriptionOLD {
     id=#58
     possessed_class=EditableTextBox
     tracks [{
       type=MovieSceneByteTrack
       display_name=Visibility
       sections [{type=MovieSceneByteSection, end_frame=8, start_frame=0, channels [None {type=MovieSceneScriptingByteChannel, const [1,1]}]}]
      }]
    }
    Circle {
     id=#30
     possessed_class=Image
     tracks [{
       type=MovieSceneByteTrack
       display_name=Visibility
       sections [{type=MovieSceneByteSection, end_frame=8, start_frame=0, channels [None {type=MovieSceneScriptingByteChannel, const [2,2]}]}]
      }]
    }
    White_Box_Solo {
     id=#59
     possessed_class=Image
     tracks [{
       type=MovieSceneByteTrack
       display_name=Visibility
       sections [{type=MovieSceneByteSection, end_frame=8, start_frame=0, channels [None {type=MovieSceneScriptingByteChannel, const [1,1]}]}]
      }]
    }
   ]}
 }
 ST_Arrow_Fade {
  binding_count=8
  length_seconds=0.45
  object=&self:ST_Arrow_Fade
  sequence {bindings [
    Image_SmallIndicator_Child {
     id=#60
     possessed_class=Image
     tracks [
      {
       type=MovieSceneByteTrack
       display_name=Visibility
       sections [{type=MovieSceneByteSection, end_frame=9, start_frame=0, channels [None {type=MovieSceneScriptingByteChannel, const [0,2]}]}]
      }
      {
       type=MovieScene2DTransformTrack
       sections [{
         type=MovieScene2DTransformSection
         end_frame=9
         start_frame=0
         channels [
          Translation.X {type=MovieSceneScriptingFloatChannel, const [0,1]}
          Translation.Y {type=MovieSceneScriptingFloatChannel, keys [[0,55],[9,75]]}
          Angle {type=MovieSceneScriptingFloatChannel, const [0,1]}
          Scale.X {type=MovieSceneScriptingFloatChannel, const [0.5,1]}
          Scale.Y {type=MovieSceneScriptingFloatChannel, const [0.5,1]}
          Shear.X {type=MovieSceneScriptingFloatChannel, const [0,1]}
          Shear.Y {type=MovieSceneScriptingFloatChannel, const [0,1]}
         ]
        }]
      }
      {
       type=MovieSceneFloatTrack
       display_name=Render Opacity
       sections [{
         type=MovieSceneFloatSection
         end_frame=9
         start_frame=0
         channels [None {type=MovieSceneScriptingFloatChannel, thinned_from=3, keys [[0,1],[9,0]]}]
        }]
      }
     ]
    }
    White_Box_Child {
     id=#61
     possessed_class=Image
     tracks [{
       type=MovieSceneByteTrack
       display_name=Visibility
       sections [{type=MovieSceneByteSection, end_frame=9, start_frame=0, channels [None {type=MovieSceneScriptingByteChannel, const [1,2]}]}]
      }]
    }
    Icon {
     id=#62
     possessed_class=Image
     tracks [{
       type=MovieSceneByteTrack
       display_name=Visibility
       sections [{type=MovieSceneByteSection, end_frame=9, start_frame=0, channels [None {type=MovieSceneScriptingByteChannel, const [2,2]}]}]
      }]
    }
    DescriptionOLD {
     id=#63
     possessed_class=EditableTextBox
     tracks [{
       type=MovieSceneByteTrack
       display_name=Visibility
       sections [{type=MovieSceneByteSection, end_frame=8, start_frame=0, channels [None {type=MovieSceneScriptingByteChannel, const [1,1]}]}]
      }]
    }
    Image_SmallIndicator {
     id=#64
     possessed_class=Image
     tracks [{
       type=MovieSceneByteTrack
       display_name=Visibility
       sections [{type=MovieSceneByteSection, end_frame=9, start_frame=0, channels [None {type=MovieSceneScriptingByteChannel, const [1,2]}]}]
      }]
    }
    Circle {
     id=#65
     possessed_class=Image
     tracks [{
       type=MovieSceneByteTrack
       display_name=Visibility
       sections [{type=MovieSceneByteSection, end_frame=9, start_frame=0, channels [None {type=MovieSceneScriptingByteChannel, const [1,2]}]}]
      }]
    }
    White_Box_Solo {
     id=#66
     possessed_class=Image
     tracks [{
       type=MovieSceneByteTrack
       display_name=Visibility
       sections [{type=MovieSceneByteSection, end_frame=8, start_frame=0, channels [None {type=MovieSceneScriptingByteChannel, const [1,1]}]}]
      }]
    }
    Sign {
     id=#67
     possessed_class=Image
     tracks [{
       type=MovieSceneByteTrack
       display_name=Visibility
       sections [{type=MovieSceneByteSection, end_frame=9, start_frame=0, channels [None {type=MovieSceneScriptingByteChannel, const [2,2]}]}]
      }]
    }
   ]}
 }
 ST_Marker_FadeIN {
  binding_count=8
  length_seconds=0.45
  object=&self:ST_Marker_FadeIN
  sequence {bindings [
    White_Box_Child {
     id=#21
     possessed_class=Image
     tracks [{
       type=MovieSceneByteTrack
       display_name=Visibility
       sections [{type=MovieSceneByteSection, end_frame=8, start_frame=0, channels [None {type=MovieSceneScriptingByteChannel, const [2,2]}]}]
      }]
    }
    Icon {
     id=#23
     possessed_class=Image
     tracks [
      {
       type=MovieSceneFloatTrack
       display_name=Render Opacity
       sections [{
         type=MovieSceneFloatSection
         end_frame=8
         start_frame=0
         channels [None {type=MovieSceneScriptingFloatChannel, keys [[0.8993,0],[2.6983,1]]}]
        }]
      }
      {
       type=MovieScene2DTransformTrack
       sections [{
         type=MovieScene2DTransformSection
         end_frame=8
         start_frame=0
         channels [
          Translation.X {type=MovieSceneScriptingFloatChannel}
          Translation.Y {type=MovieSceneScriptingFloatChannel}
          Angle {type=MovieSceneScriptingFloatChannel}
          Scale.X {type=MovieSceneScriptingFloatChannel, keys [[0.8993,1.2],[3.598,0.7],[6.0697,1.2],[7.475,1]]}
          Scale.Y {type=MovieSceneScriptingFloatChannel, keys [[0.8993,0.7],[3.598,1.3],[6.0697,0.8],[7.475,1]]}
          Shear.X {type=MovieSceneScriptingFloatChannel}
          Shear.Y {type=MovieSceneScriptingFloatChannel}
         ]
        }]
      }
     ]
    }
    White_Box_Solo {
     id=#68
     possessed_class=Image
     tracks [{
       type=MovieSceneByteTrack
       display_name=Visibility
       sections [{type=MovieSceneByteSection, end_frame=8, start_frame=0, channels [None {type=MovieSceneScriptingByteChannel, const [1,1]}]}]
      }]
    }
    DescriptionOLD {
     id=#69
     possessed_class=EditableTextBox
     tracks [{
       type=MovieSceneByteTrack
       display_name=Visibility
       sections [{type=MovieSceneByteSection, end_frame=8, start_frame=0, channels [None {type=MovieSceneScriptingByteChannel, const [1,1]}]}]
      }]
    }
    Image_SmallIndicator_Child {
     id=#44
     possessed_class=Image
     tracks [{
       type=MovieSceneByteTrack
       display_name=Visibility
       sections [{type=MovieSceneByteSection, end_frame=8, start_frame=0, channels [None {type=MovieSceneScriptingByteChannel, const [1,2]}]}]
      }]
    }
    Sign {
     id=#29
     possessed_class=Image
     tracks [
      {
       type=MovieScene2DTransformTrack
       sections [{
         type=MovieScene2DTransformSection
         end_frame=8
         start_frame=-1
         channels [
          Scale.X {type=MovieSceneScriptingFloatChannel, keys [[1.7987,1],[3.597,0.4],[5.1993,1],[6.484,1.2],[7.0817,1],[8.995,1]]}
          Scale.Y {type=MovieSceneScriptingFloatChannel, keys [[1.7987,1],[3.597,1.6],[5.1993,1],[6.4843,0.8],[7.082,1],[8.995,1]]}
         ]
        }]
      }
      {
       type=MovieSceneFloatTrack
       display_name=Render Opacity
       sections [{
         type=MovieSceneFloatSection
         end_frame=8
         start_frame=0
         channels [None {type=MovieSceneScriptingFloatChannel, keys [[0.8993,0],[1.7983,0],[3.598,1]]}]
        }]
      }
     ]
    }
    Circle {
     id=#30
     possessed_class=Image
     tracks [{
       type=MovieSceneByteTrack
       display_name=Visibility
       sections [{type=MovieSceneByteSection, end_frame=8, start_frame=0, channels [None {type=MovieSceneScriptingByteChannel, const [2,2]}]}]
      }]
    }
    Image_SmallIndicator {
     id=#47
     possessed_class=Image
     tracks [
      {
       type=MovieScene2DTransformTrack
       sections [{
         type=MovieScene2DTransformSection
         end_frame=9
         start_frame=-1
         channels [
          Translation.X {type=MovieSceneScriptingFloatChannel}
          Translation.Y {type=MovieSceneScriptingFloatChannel, const [45,3]}
          Angle {type=MovieSceneScriptingFloatChannel}
          Scale.X {type=MovieSceneScriptingFloatChannel, keys [[-0.001,1.2],[0.8987,1.2],[3.598,0.7],[7.1973,1.2],[8.5423,1],[8.9963,1]]}
          Scale.Y {type=MovieSceneScriptingFloatChannel, keys [[-0.001,0.7],[0.8987,0.7],[3.598,1.3],[7.1973,0.8],[8.5423,1],[8.9963,1]]}
          Shear.X {type=MovieSceneScriptingFloatChannel}
          Shear.Y {type=MovieSceneScriptingFloatChannel}
         ]
        }]
      }
      {
       type=MovieSceneFloatTrack
       display_name=Render Opacity
       sections [{
         type=MovieSceneFloatSection
         end_frame=8
         start_frame=0
         channels [None {type=MovieSceneScriptingFloatChannel, keys [[0.8993,0],[2.6983,1]]}]
        }]
      }
     ]
    }
   ]}
 }
 ST_Out_Arrow_Marker {
  binding_count=9
  length_seconds=0.65
  object=&self:ST_Out_Arrow_Marker
  sequence {bindings [
    White_Box_Child {
     id=#21
     possessed_class=Image
     tracks [
      {
       type=MovieSceneFloatTrack
       display_name=Render Opacity
       sections [{type=MovieSceneFloatSection, end_frame=13, start_frame=-1, channels [None {type=MovieSceneScriptingFloatChannel, const [1,1]}]}]
      }
      {
       type=MovieSceneByteTrack
       display_name=Visibility
       sections [{type=MovieSceneByteSection, end_frame=13, start_frame=-1, channels [None {type=MovieSceneScriptingByteChannel, const [2,1]}]}]
      }
     ]
    }
    Icon {
     id=#23
     possessed_class=Image
     tracks [{
       type=MovieSceneFloatTrack
       display_name=Render Opacity
       sections [{type=MovieSceneFloatSection, end_frame=13, start_frame=0, channels [None {type=MovieSceneScriptingFloatChannel, const [0,1]}]}]
      }]
    }
    DescriptionOLD {
     id=#70
     possessed_class=EditableTextBox
     tracks [{
       type=MovieSceneByteTrack
       display_name=Visibility
       sections [{type=MovieSceneByteSection, end_frame=13, start_frame=-1, channels [None {type=MovieSceneScriptingByteChannel, const [1,1]}]}]
      }]
    }
    Overlay_Icon_Box {
     id=#28
     possessed_class=Overlay
     tracks [{
       type=MovieScene2DTransformTrack
       sections [{
         type=MovieScene2DTransformSection
         end_frame=12
         start_frame=-1
         channels [
          Translation.X {type=MovieSceneScriptingFloatChannel}
          Translation.Y {type=MovieSceneScriptingFloatChannel}
          Angle {type=MovieSceneScriptingFloatChannel, const [0,1]}
          Scale.X {type=MovieSceneScriptingFloatChannel, const [1,1]}
          Scale.Y {type=MovieSceneScriptingFloatChannel, const [1,1]}
          Shear.X {type=MovieSceneScriptingFloatChannel}
          Shear.Y {type=MovieSceneScriptingFloatChannel}
         ]
        }]
      }]
    }
    Image_SmallIndicator {
     id=#56
     possessed_class=Image
     tracks [
      {
       type=MovieScene2DTransformTrack
       sections [{
         type=MovieScene2DTransformSection
         end_frame=13
         start_frame=-1
         channels [
          Translation.X {type=MovieSceneScriptingFloatChannel, const [0,1]}
          Translation.Y {type=MovieSceneScriptingFloatChannel, const [55,1]}
          Angle {type=MovieSceneScriptingFloatChannel, const [0,1]}
          Scale.X {type=MovieSceneScriptingFloatChannel, const [0.5,1]}
          Scale.Y {type=MovieSceneScriptingFloatChannel, const [0.5,1]}
          Shear.X {type=MovieSceneScriptingFloatChannel, const [0,1]}
          Shear.Y {type=MovieSceneScriptingFloatChannel, const [0,1]}
         ]
        }]
      }
      {
       type=MovieSceneFloatTrack
       display_name=Render Opacity
       sections [{type=MovieSceneFloatSection, end_frame=13, start_frame=0, channels [None {type=MovieSceneScriptingFloatChannel, const [0,1]}]}]
      }
     ]
    }
    Image_SmallIndicator_Child {
     id=#57
     possessed_class=Image
     tracks [{
       type=MovieSceneByteTrack
       display_name=Visibility
       sections [{type=MovieSceneByteSection, end_frame=10, start_frame=0, channels [None {type=MovieSceneScriptingByteChannel, const [1,1]}]}]
      }]
    }
    Sign {
     id=#29
     possessed_class=Image
     tracks [
      {
       type=MovieSceneFloatTrack
       display_name=Render Opacity
       sections [{
         type=MovieSceneFloatSection
         end_frame=13
         start_frame=0
         channels [None {type=MovieSceneScriptingFloatChannel, keys [[0.0003,1],[10,1],[12.9997,0]]}]
        }]
      }
      {
       type=MovieScene2DTransformTrack
       sections [{
         type=MovieScene2DTransformSection
         end_frame=13
         start_frame=-1
         channels [
          Translation.X {type=MovieSceneScriptingFloatChannel, const [0,1]}
          Translation.Y {type=MovieSceneScriptingFloatChannel, const [0,1]}
          Angle {type=MovieSceneScriptingFloatChannel, const [0,1]}
          Scale.X {type=MovieSceneScriptingFloatChannel, const [1,2]}
          Scale.Y {type=MovieSceneScriptingFloatChannel, keys [[0.0003,1],[9.9997,1],[12.9997,0.74]]}
          Shear.X {type=MovieSceneScriptingFloatChannel, const [0,1]}
          Shear.Y {type=MovieSceneScriptingFloatChannel, const [0,1]}
         ]
        }]
      }
     ]
    }
    Circle {
     id=#30
     possessed_class=Image
     tracks [{
       type=MovieSceneByteTrack
       display_name=Visibility
       sections [{type=MovieSceneByteSection, end_frame=13, start_frame=0, channels [None {type=MovieSceneScriptingByteChannel, const [2,1]}]}]
      }]
    }
    White_Box_Solo {
     id=#71
     possessed_class=Image
     tracks [{
       type=MovieSceneByteTrack
       display_name=Visibility
       sections [{type=MovieSceneByteSection, end_frame=13, start_frame=-1, channels [None {type=MovieSceneScriptingByteChannel, const [1,1]}]}]
      }]
    }
   ]}
 }
 ST_Arrow_Transition {
  binding_count=9
  length_seconds=0.5
  object=&self:ST_Arrow_Transition
  sequence {bindings [
    White_Box_Child {
     id=#21
     possessed_class=Image
     tracks [{
       type=MovieSceneByteTrack
       display_name=Visibility
       sections [{type=MovieSceneByteSection, end_frame=9, start_frame=0, channels [None {type=MovieSceneScriptingByteChannel, const [2,1]}]}]
      }]
    }
    Icon {
     id=#23
     possessed_class=Image
     tracks [
      {
       type=MovieSceneFloatTrack
       display_name=Render Opacity
       sections [{type=MovieSceneFloatSection, end_frame=10, start_frame=0, channels [None {type=MovieSceneScriptingFloatChannel, const [1,1]}]}]
      }
      {
       type=MovieScene2DTransformTrack
       sections [{
         type=MovieScene2DTransformSection
         end_frame=9
         start_frame=-1
         channels [
          Translation.X {type=MovieSceneScriptingFloatChannel, const [0,1]}
          Translation.Y {type=MovieSceneScriptingFloatChannel, const [0,1]}
          Angle {type=MovieSceneScriptingFloatChannel, const [0,1]}
          Scale.X {type=MovieSceneScriptingFloatChannel, const [1,1]}
          Scale.Y {type=MovieSceneScriptingFloatChannel, const [1,1]}
          Shear.X {type=MovieSceneScriptingFloatChannel, const [0,1]}
          Shear.Y {type=MovieSceneScriptingFloatChannel, const [0,1]}
         ]
        }]
      }
     ]
    }
    DescriptionOLD {
     id=#72
     possessed_class=EditableTextBox
     tracks [{
       type=MovieSceneByteTrack
       display_name=Visibility
       sections [{type=MovieSceneByteSection, end_frame=9, start_frame=-1, channels [None {type=MovieSceneScriptingByteChannel, const [1,1]}]}]
      }]
    }
    Image_SmallIndicator_Child {
     id=#44
     possessed_class=Image
     tracks [{
       type=MovieSceneByteTrack
       display_name=Visibility
       sections [{type=MovieSceneByteSection, end_frame=9, start_frame=0, channels [None {type=MovieSceneScriptingByteChannel, const [1,1]}]}]
      }]
    }
    Sign {
     id=#29
     possessed_class=Image
     tracks [
      {
       type=MovieScene2DTransformTrack
       sections [{
         type=MovieScene2DTransformSection
         end_frame=9
         start_frame=0
         channels [
          Scale.X {type=MovieSceneScriptingFloatChannel, keys [[0,1],[3.8563,0.8],[7.0707,1]]}
          Scale.Y {type=MovieSceneScriptingFloatChannel, keys [[0,1],[1.285,1],[3.8563,1.25],[7.0707,1]]}
         ]
        }]
      }
      {
       type=MovieSceneFloatTrack
       display_name=Render Opacity
       sections [{type=MovieSceneFloatSection, end_frame=9, start_frame=0, channels [None {type=MovieSceneScriptingFloatChannel, const [1,1]}]}]
      }
     ]
    }
    Circle {
     id=#30
     possessed_class=Image
     tracks [{
       type=MovieSceneByteTrack
       display_name=Visibility
       sections [{type=MovieSceneByteSection, end_frame=9, start_frame=0, channels [None {type=MovieSceneScriptingByteChannel, const [2,1]}]}]
      }]
    }
    White_Box_Solo {
     id=#73
     possessed_class=Image
     tracks [{
       type=MovieSceneByteTrack
       display_name=Visibility
       sections [{type=MovieSceneByteSection, end_frame=9, start_frame=-1, channels [None {type=MovieSceneScriptingByteChannel, const [1,1]}]}]
      }]
    }
    Image_SmallIndicator {
     id=#47
     possessed_class=Image
     tracks [
      {
       type=MovieScene2DTransformTrack
       sections [{
         type=MovieScene2DTransformSection
         end_frame=9
         start_frame=0
         channels [
          Translation.X {type=MovieSceneScriptingFloatChannel, const [0,3]}
          Translation.Y {type=MovieSceneScriptingFloatChannel, keys [[0,45],[3.8563,45],[7.0707,55]]}
          Angle {type=MovieSceneScriptingFloatChannel, const [0,3]}
          Scale.X {type=MovieSceneScriptingFloatChannel, keys [[0,1],[3.8563,1],[7.0707,0.5]]}
          Scale.Y {type=MovieSceneScriptingFloatChannel, keys [[0,1],[3.8563,1],[7.0707,0.5]]}
          Shear.X {type=MovieSceneScriptingFloatChannel, const [0,3]}
          Shear.Y {type=MovieSceneScriptingFloatChannel, const [0,3]}
         ]
        }]
      }
      {
       type=MovieSceneFloatTrack
       display_name=Render Opacity
       sections [{
         type=MovieSceneFloatSection
         end_frame=9
         start_frame=0
         channels [None {type=MovieSceneScriptingFloatChannel, keys [[0,1],[1.285,1],[3.8563,0],[7.0707,0],[8.999,1]]}]
        }]
      }
     ]
    }
    Overlay_Icon {
     id=#74
     possessed_class=Overlay
     tracks [{
       type=MovieScene2DTransformTrack
       sections [{
         type=MovieScene2DTransformSection
         end_frame=9
         start_frame=0
         channels [
          Translation.X {type=MovieSceneScriptingFloatChannel, const [0,4]}
          Translation.Y {type=MovieSceneScriptingFloatChannel, const [0,4]}
          Angle {type=MovieSceneScriptingFloatChannel, const [0,4]}
          Scale.X {type=MovieSceneScriptingFloatChannel, keys [[1.285,1],[3.8563,0.7],[5.7847,1],[7.0707,1]]}
          Scale.Y {type=MovieSceneScriptingFloatChannel, keys [[1.285,1],[3.8563,1.3],[5.7847,0.8],[7.0707,1]]}
          Shear.X {type=MovieSceneScriptingFloatChannel, const [0,4]}
          Shear.Y {type=MovieSceneScriptingFloatChannel, const [0,4]}
         ]
        }]
      }]
    }
   ]}
 }
 ST_Arrow_Bounce {
  binding_count=10
  length_seconds=0.45
  object=&self:ST_Arrow_Bounce
  sequence {bindings [
    White_Box_Solo {
     id=#75
     possessed_class=Image
     tracks [{
       type=MovieSceneByteTrack
       display_name=Visibility
       sections [{type=MovieSceneByteSection, end_frame=9, start_frame=-1, channels [None {type=MovieSceneScriptingByteChannel, const [1,1]}]}]
      }]
    }
    White_Box_Child {
     id=#21
     possessed_class=Image
     tracks [{
       type=MovieSceneByteTrack
       display_name=Visibility
       sections [{type=MovieSceneByteSection, end_frame=8, start_frame=-1, channels [None {type=MovieSceneScriptingByteChannel, const [2,2]}]}]
      }]
    }
    Icon {
     id=#23
     possessed_class=Image
     tracks [
      {
       type=MovieSceneFloatTrack
       display_name=Render Opacity
       sections [{type=MovieSceneFloatSection, end_frame=9, start_frame=-1, channels [None {type=MovieSceneScriptingFloatChannel, const [1,2]}]}]
      }
      {
       type=MovieScene2DTransformTrack
       sections [{
         type=MovieScene2DTransformSection
         end_frame=8
         start_frame=-1
         channels [
          Translation.X {type=MovieSceneScriptingFloatChannel, const [0,2]}
          Translation.Y {type=MovieSceneScriptingFloatChannel, const [0,2]}
          Angle {type=MovieSceneScriptingFloatChannel, const [0,2]}
          Scale.X {type=MovieSceneScriptingFloatChannel, const [1,1]}
          Scale.Y {type=MovieSceneScriptingFloatChannel, const [1,1]}
          Shear.X {type=MovieSceneScriptingFloatChannel, const [0,2]}
          Shear.Y {type=MovieSceneScriptingFloatChannel, const [0,2]}
         ]
        }]
      }
     ]
    }
    DescriptionOLD {
     id=#76
     possessed_class=EditableTextBox
     tracks [{
       type=MovieSceneByteTrack
       display_name=Visibility
       sections [{type=MovieSceneByteSection, end_frame=8, start_frame=-1, channels [None {type=MovieSceneScriptingByteChannel, const [1,1]}]}]
      }]
    }
    Image_SmallIndicator_Child {
     id=#44
     possessed_class=Image
     tracks [{
       type=MovieSceneByteTrack
       display_name=Visibility
       sections [{type=MovieSceneByteSection, end_frame=6, start_frame=-1, channels [None {type=MovieSceneScriptingByteChannel, const [1,2]}]}]
      }]
    }
    Sign {
     id=#29
     possessed_class=Image
     tracks [
      {
       type=MovieScene2DTransformTrack
       sections [{
         type=MovieScene2DTransformSection
         end_frame=8
         start_frame=-1
         channels [
          Scale.X {type=MovieSceneScriptingFloatChannel, keys [[-0.0003,1],[2.2493,1],[4.4993,0.65],[5.999,1.2],[6.749,1],[8.999,1]]}
          Scale.Y {type=MovieSceneScriptingFloatChannel, keys [[-0.0003,1],[2.2493,1],[4.4993,1.35],[5.999,0.8],[6.749,1],[8.999,1]]}
         ]
        }]
      }
      {
       type=MovieSceneFloatTrack
       display_name=Render Opacity
       sections [{type=MovieSceneFloatSection, end_frame=9, start_frame=0, channels [None {type=MovieSceneScriptingFloatChannel, const [1,1]}]}]
      }
     ]
    }
    Circle {
     id=#30
     possessed_class=Image
     tracks [{
       type=MovieSceneByteTrack
       display_name=Visibility
       sections [{type=MovieSceneByteSection, end_frame=8, start_frame=-1, channels [None {type=MovieSceneScriptingByteChannel, const [2,2]}]}]
      }]
    }
    Image_SmallIndicator {
     id=#47
     possessed_class=Image
     tracks [
      {
       type=MovieScene2DTransformTrack
       sections [{
         type=MovieScene2DTransformSection
         end_frame=8
         start_frame=-1
         channels [
          Scale.X {type=MovieSceneScriptingFloatChannel, keys [[-0.0003,1.7],[2.087,0.8],[3.4803,1.1],[5.2217,0.95],[6.9633,1]]}
          Scale.Y {type=MovieSceneScriptingFloatChannel, keys [[-0.0003,0],[2.087,1.2],[3.4803,0.9],[5.2217,1.05],[6.9633,1]]}
         ]
        }]
      }
      {
       type=MovieSceneFloatTrack
       display_name=Render Opacity
       sections [{type=MovieSceneFloatSection, end_frame=8, start_frame=-1, channels [None {type=MovieSceneScriptingFloatChannel, const [1,2]}]}]
      }
      {
       type=MovieSceneColorTrack
       display_name=Color and Opacity
       sections [{
         type=MovieSceneColorSection
         end_frame=2
         start_frame=-1
         channels [
          Color.R {type=MovieSceneScriptingFloatChannel, const [1,1]}
          Color.G {type=MovieSceneScriptingFloatChannel, const [1,1]}
          Color.B {type=MovieSceneScriptingFloatChannel, const [1,1]}
          Color.A {type=MovieSceneScriptingFloatChannel, keys [[0,0],[2.1093,1]]}
         ]
        }]
      }
      {
       type=MovieSceneByteTrack
       display_name=Visibility
       sections [{type=MovieSceneByteSection, end_frame=0, start_frame=-1, channels [None {type=MovieSceneScriptingByteChannel, const [4,1]}]}]
      }
     ]
    }
    Overlay_Icon {
     id=#74
     possessed_class=Overlay
     tracks [{
       type=MovieScene2DTransformTrack
       sections [{
         type=MovieScene2DTransformSection
         end_frame=9
         start_frame=-1
         channels [
          Translation.X {type=MovieSceneScriptingFloatChannel, const [0,3]}
          Translation.Y {type=MovieSceneScriptingFloatChannel, const [0,3]}
          Angle {type=MovieSceneScriptingFloatChannel, const [0,3]}
          Scale.X {type=MovieSceneScriptingFloatChannel, keys [[-0.0003,1],[1.4993,1],[3.7493,0.7],[5.5737,1.2],[6.3237,1],[8.999,1]]}
          Scale.Y {type=MovieSceneScriptingFloatChannel, keys [[-0.0003,1],[1.4993,1],[3.7493,1.3],[5.5737,0.8],[6.3237,1],[8.999,1]]}
         ]
        }]
      }]
    }
   ]}
 }
 ST_Arrow_Disappear {
  binding_count=9
  length_seconds=0.5
  object=&self:ST_Arrow_Disappear
  sequence {bindings [
    White_Box_Child {
     id=#21
     possessed_class=Image
     tracks [{
       type=MovieSceneByteTrack
       display_name=Visibility
       sections [{type=MovieSceneByteSection, end_frame=8, start_frame=0, channels [None {type=MovieSceneScriptingByteChannel, const [2,1]}]}]
      }]
    }
    White_Box_Solo {
     id=#77
     possessed_class=Image
     tracks [{
       type=MovieSceneByteTrack
       display_name=Visibility
       sections [{type=MovieSceneByteSection, end_frame=8, start_frame=-1, channels [None {type=MovieSceneScriptingByteChannel, const [1,1]}]}]
      }]
    }
    Icon {
     id=#23
     possessed_class=Image
     tracks [
      {
       type=MovieSceneFloatTrack
       display_name=Render Opacity
       sections [{type=MovieSceneFloatSection, end_frame=8, start_frame=0, channels [None {type=MovieSceneScriptingFloatChannel, const [1,1]}]}]
      }
      {
       type=MovieScene2DTransformTrack
       sections [{
         type=MovieScene2DTransformSection
         end_frame=8
         start_frame=-1
         channels [
          Translation.X {type=MovieSceneScriptingFloatChannel, const [0,1]}
          Translation.Y {type=MovieSceneScriptingFloatChannel, const [0,1]}
          Angle {type=MovieSceneScriptingFloatChannel, const [0,1]}
          Scale.X {type=MovieSceneScriptingFloatChannel, const [1,1]}
          Scale.Y {type=MovieSceneScriptingFloatChannel, const [1,1]}
          Shear.X {type=MovieSceneScriptingFloatChannel, const [0,1]}
          Shear.Y {type=MovieSceneScriptingFloatChannel, const [0,1]}
         ]
        }]
      }
     ]
    }
    Image_SmallIndicator_Child {
     id=#44
     possessed_class=Image
     tracks [{
       type=MovieSceneByteTrack
       display_name=Visibility
       sections [{type=MovieSceneByteSection, end_frame=7, start_frame=0, channels [None {type=MovieSceneScriptingByteChannel, const [1,1]}]}]
      }]
    }
    Sign {
     id=#29
     possessed_class=Image
     tracks [
      {
       type=MovieScene2DTransformTrack
       sections [{
         type=MovieScene2DTransformSection
         end_frame=8
         start_frame=0
         channels [
          Scale.X {type=MovieSceneScriptingFloatChannel, keys [[0,1],[4.666,0.8],[6.4283,1.2],[7.3327,1]]}
          Scale.Y {type=MovieSceneScriptingFloatChannel, keys [[0,1],[1.9997,1],[4.666,1.25],[6.4283,0.8],[7.3327,1]]}
         ]
        }]
      }
      {
       type=MovieSceneFloatTrack
       display_name=Render Opacity
       sections [{type=MovieSceneFloatSection, end_frame=10, start_frame=0, channels [None {type=MovieSceneScriptingFloatChannel, const [1,1]}]}]
      }
     ]
    }
    Circle {
     id=#30
     possessed_class=Image
     tracks [{
       type=MovieSceneByteTrack
       display_name=Visibility
       sections [{type=MovieSceneByteSection, end_frame=8, start_frame=0, channels [None {type=MovieSceneScriptingByteChannel, const [2,1]}]}]
      }]
    }
    DescriptionOLD {
     id=#78
     possessed_class=EditableTextBox
     tracks [{
       type=MovieSceneByteTrack
       display_name=Visibility
       sections [{type=MovieSceneByteSection, end_frame=8, start_frame=-1, channels [None {type=MovieSceneScriptingByteChannel, const [1,1]}]}]
      }]
    }
    Image_SmallIndicator {
     id=#47
     possessed_class=Image
     tracks [
      {
       type=MovieScene2DTransformTrack
       sections [{
         type=MovieScene2DTransformSection
         end_frame=8
         start_frame=0
         channels [
          Translation.X {type=MovieSceneScriptingFloatChannel, const [0,2]}
          Translation.Y {type=MovieSceneScriptingFloatChannel, const [45,2]}
          Angle {type=MovieSceneScriptingFloatChannel, const [0,2]}
          Scale.X {type=MovieSceneScriptingFloatChannel, const [1,2]}
          Scale.Y {type=MovieSceneScriptingFloatChannel, const [1,2]}
          Shear.X {type=MovieSceneScriptingFloatChannel, const [0,2]}
          Shear.Y {type=MovieSceneScriptingFloatChannel, const [0,2]}
         ]
        }]
      }
      {
       type=MovieSceneFloatTrack
       display_name=Render Opacity
       sections [{
         type=MovieSceneFloatSection
         end_frame=8
         start_frame=0
         channels [None {type=MovieSceneScriptingFloatChannel, keys [[0,1],[1.333,1],[3.9993,0]]}]
        }]
      }
     ]
    }
    Overlay_Icon {
     id=#74
     possessed_class=Overlay
     tracks [{
       type=MovieScene2DTransformTrack
       sections [{
         type=MovieScene2DTransformSection
         end_frame=7
         start_frame=0
         channels [
          Translation.X {type=MovieSceneScriptingFloatChannel, const [0,3]}
          Translation.Y {type=MovieSceneScriptingFloatChannel, const [0,3]}
          Angle {type=MovieSceneScriptingFloatChannel, const [0,3]}
          Scale.X {type=MovieSceneScriptingFloatChannel, keys [[1.2723,1],[3.8177,0.6],[5.7267,1.3],[6.9993,1]]}
          Scale.Y {type=MovieSceneScriptingFloatChannel, keys [[1.2723,1],[3.8177,1.4],[5.7267,0.7],[6.9993,1]]}
          Shear.X {type=MovieSceneScriptingFloatChannel, const [0,3]}
          Shear.Y {type=MovieSceneScriptingFloatChannel, const [0,3]}
         ]
        }]
      }]
    }
   ]}
 }
 ST_Text_in {
  binding_count=1
  length_seconds=0.4
  object=&self:ST_Text_in
  sequence {bindings [DescriptionOLD {
     id=#79
     display_name=Description
     possessed_class=EditableTextBox
     tracks [
      {
       type=MovieSceneByteTrack
       display_name=Visibility
       sections [{type=MovieSceneByteSection, end_frame=7, start_frame=0, channels [None {type=MovieSceneScriptingByteChannel, const [0,1]}]}]
      }
      {
       type=MovieScene2DTransformTrack
       sections [{
         type=MovieScene2DTransformSection
         end_frame=8
         start_frame=0
         channels [
          Translation.X {type=MovieSceneScriptingFloatChannel, const [70,3]}
          Translation.Y {type=MovieSceneScriptingFloatChannel}
          Angle {type=MovieSceneScriptingFloatChannel}
          Scale.X {type=MovieSceneScriptingFloatChannel, keys [[0,0],[2.8847,1.3],[4.7157,0.8],[5.7137,1]]}
          Scale.Y {type=MovieSceneScriptingFloatChannel, keys [[0,1.4],[2.8847,0.7],[4.7157,1.2],[5.7137,1]]}
          Shear.X {type=MovieSceneScriptingFloatChannel}
          Shear.Y {type=MovieSceneScriptingFloatChannel}
         ]
        }]
      }
      {
       type=MovieSceneFloatTrack
       display_name=Render Opacity
       sections [{type=MovieSceneFloatSection, end_frame=7, start_frame=0, channels [None {type=MovieSceneScriptingFloatChannel, keys [[0,0],[2.8847,1]]}]}]
      }
     ]
    }]}
 }
 ST_Text_Out {
  binding_count=1
  length_seconds=0.5
  object=&self:ST_Text_Out
  sequence {bindings [DescriptionOLD {
     id=#79
     display_name=Description
     possessed_class=EditableTextBox
     tracks [
      {
       type=MovieSceneByteTrack
       display_name=Visibility
       sections [{type=MovieSceneByteSection, end_frame=10, start_frame=0, channels [None {type=MovieSceneScriptingByteChannel, const [0,1]}]}]
      }
      {
       type=MovieScene2DTransformTrack
       sections [{
         type=MovieScene2DTransformSection
         end_frame=10
         start_frame=0
         channels [
          Translation.X {type=MovieSceneScriptingFloatChannel, keys [[0,70],[5,90]]}
          Translation.Y {type=MovieSceneScriptingFloatChannel}
         ]
        }]
      }
      {
       type=MovieSceneFloatTrack
       display_name=Render Opacity
       sections [{type=MovieSceneFloatSection, end_frame=10, start_frame=0, channels [None {type=MovieSceneScriptingFloatChannel, keys [[0,1],[5,0]]}]}]
      }
     ]
    }]}
 }
 ST_Marker_Out01 {
  binding_count=8
  length_seconds=0.5001
  object=&self:ST_Marker_Out01
  sequence {bindings [
    White_Box_Child {
     id=#21
     possessed_class=Image
     tracks [{
       type=MovieSceneByteTrack
       display_name=Visibility
       sections [{type=MovieSceneByteSection, end_frame=9, start_frame=0, channels [None {type=MovieSceneScriptingByteChannel, const [2,1]}]}]
      }]
    }
    Icon {
     id=#23
     possessed_class=Image
     tracks [
      {
       type=MovieSceneFloatTrack
       display_name=Render Opacity
       sections [{type=MovieSceneFloatSection, end_frame=10, start_frame=0, channels [None {type=MovieSceneScriptingFloatChannel, keys [[4,1],[9,0]]}]}]
      }
      {
       type=MovieScene2DTransformTrack
       sections [{
         type=MovieScene2DTransformSection
         end_frame=10
         start_frame=0
         channels [
          Translation.X {type=MovieSceneScriptingFloatChannel, const [0,1]}
          Translation.Y {type=MovieSceneScriptingFloatChannel, const [0,1]}
          Angle {type=MovieSceneScriptingFloatChannel}
          Scale.X {type=MovieSceneScriptingFloatChannel, keys [[4,1],[5.3857,2.2],[9,0]]}
          Scale.Y {type=MovieSceneScriptingFloatChannel, keys [[4,1],[5.3857,2.2],[9,0]]}
          Shear.X {type=MovieSceneScriptingFloatChannel}
          Shear.Y {type=MovieSceneScriptingFloatChannel}
         ]
        }]
      }
     ]
    }
    White_Box_Solo {
     id=#68
     possessed_class=Image
     tracks [{
       type=MovieSceneByteTrack
       display_name=Visibility
       sections [{type=MovieSceneByteSection, end_frame=9, start_frame=0, channels [None {type=MovieSceneScriptingByteChannel, const [1,1]}]}]
      }]
    }
    DescriptionOLD {
     id=#69
     possessed_class=EditableTextBox
     tracks [
      {
       type=MovieSceneByteTrack
       display_name=Visibility
       sections [{type=MovieSceneByteSection, end_frame=10, start_frame=-1, channels [None {type=MovieSceneScriptingByteChannel, const [4,1]}]}]
      }
      {
       type=MovieScene2DTransformTrack
       sections [{
         type=MovieScene2DTransformSection
         end_frame=10
         start_frame=-1
         channels [
          Translation.X {type=MovieSceneScriptingFloatChannel, keys [[-0.0013,0],[4.4523,130],[10,185]]}
          Translation.Y {type=MovieSceneScriptingFloatChannel, const [0,2]}
          Angle {type=MovieSceneScriptingFloatChannel, const [0,1]}
          Scale.X {type=MovieSceneScriptingFloatChannel, const [1,1]}
          Scale.Y {type=MovieSceneScriptingFloatChannel, const [1,1]}
          Shear.X {type=MovieSceneScriptingFloatChannel, const [0,1]}
          Shear.Y {type=MovieSceneScriptingFloatChannel, const [0,1]}
         ]
        }]
      }
      {
       type=MovieSceneFloatTrack
       display_name=Render Opacity
       sections [{
         type=MovieSceneFloatSection
         end_frame=10
         start_frame=0
         channels [None {type=MovieSceneScriptingFloatChannel, keys [[0,1],[4.4523,1],[10,0]]}]
        }]
      }
     ]
    }
    Image_SmallIndicator_Child {
     id=#44
     possessed_class=Image
     tracks [{
       type=MovieSceneByteTrack
       display_name=Visibility
       sections [{type=MovieSceneByteSection, end_frame=9, start_frame=0, channels [None {type=MovieSceneScriptingByteChannel, const [1,1]}]}]
      }]
    }
    Sign {
     id=#29
     possessed_class=Image
     tracks [
      {
       type=MovieScene2DTransformTrack
       sections [{
         type=MovieScene2DTransformSection
         end_frame=10
         start_frame=-1
         channels [
          Scale.X {type=MovieSceneScriptingFloatChannel, keys [[0.0003,1],[3,0.5],[5.4607,2]]}
          Scale.Y {type=MovieSceneScriptingFloatChannel, keys [[0.0003,1],[3,1.5],[5.4607,0]]}
         ]
        }]
      }
      {
       type=MovieSceneFloatTrack
       display_name=Render Opacity
       sections [{type=MovieSceneFloatSection, end_frame=10, start_frame=0, channels [None {type=MovieSceneScriptingFloatChannel, const [1,1]}]}]
      }
     ]
    }
    Circle {
     id=#30
     possessed_class=Image
     tracks [{
       type=MovieSceneByteTrack
       display_name=Visibility
       sections [{type=MovieSceneByteSection, end_frame=9, start_frame=0, channels [None {type=MovieSceneScriptingByteChannel, const [2,1]}]}]
      }]
    }
    Image_SmallIndicator {
     id=#47
     possessed_class=Image
     tracks [
      {
       type=MovieScene2DTransformTrack
       sections [{
         type=MovieScene2DTransformSection
         end_frame=10
         start_frame=-1
         channels [
          Translation.X {type=MovieSceneScriptingFloatChannel, const [0,1]}
          Translation.Y {type=MovieSceneScriptingFloatChannel, keys [[0,45],[3.9893,45],[5.3857,65],[10,75]]}
          Angle {type=MovieSceneScriptingFloatChannel}
          Scale.X {type=MovieSceneScriptingFloatChannel, keys [[0,1],[3.9893,1],[10,1.4]]}
          Scale.Y {type=MovieSceneScriptingFloatChannel, keys [[0,1],[3.9893,1],[10,1.4]]}
          Shear.X {type=MovieSceneScriptingFloatChannel}
          Shear.Y {type=MovieSceneScriptingFloatChannel}
         ]
        }]
      }
      {
       type=MovieSceneFloatTrack
       display_name=Render Opacity
       sections [{type=MovieSceneFloatSection, end_frame=10, start_frame=0, channels [None {type=MovieSceneScriptingFloatChannel, keys [[3.6143,1],[9,0]]}]}]
      }
      {
       type=MovieSceneByteTrack
       display_name=Visibility
       sections [{type=MovieSceneByteSection, end_frame=10, start_frame=-1, channels [None {type=MovieSceneScriptingByteChannel, const [0,1]}]}]
      }
     ]
    }
   ]}
 }
 ST_Marker_Out02 {
  binding_count=8
  length_seconds=0.5001
  object=&self:ST_Marker_Out02
  sequence {bindings [
    White_Box_Child {
     id=#21
     possessed_class=Image
     tracks [{
       type=MovieSceneByteTrack
       display_name=Visibility
       sections [{type=MovieSceneByteSection, end_frame=9, start_frame=0, channels [None {type=MovieSceneScriptingByteChannel, const [2,1]}]}]
      }]
    }
    Icon {
     id=#23
     possessed_class=Image
     tracks [
      {
       type=MovieSceneFloatTrack
       display_name=Render Opacity
       sections [{type=MovieSceneFloatSection, end_frame=10, start_frame=0, channels [None {type=MovieSceneScriptingFloatChannel, keys [[4,1],[9,0]]}]}]
      }
      {
       type=MovieScene2DTransformTrack
       sections [{
         type=MovieScene2DTransformSection
         end_frame=10
         start_frame=0
         channels [
          Translation.X {type=MovieSceneScriptingFloatChannel, const [0,1]}
          Translation.Y {type=MovieSceneScriptingFloatChannel, const [0,1]}
          Angle {type=MovieSceneScriptingFloatChannel}
          Scale.X {type=MovieSceneScriptingFloatChannel, keys [[4,1],[5.3857,2.2],[9,0]]}
          Scale.Y {type=MovieSceneScriptingFloatChannel, keys [[4,1],[5.3857,2.2],[9,0]]}
          Shear.X {type=MovieSceneScriptingFloatChannel}
          Shear.Y {type=MovieSceneScriptingFloatChannel}
         ]
        }]
      }
     ]
    }
    White_Box_Solo {
     id=#68
     possessed_class=Image
     tracks [{
       type=MovieSceneByteTrack
       display_name=Visibility
       sections [{type=MovieSceneByteSection, end_frame=9, start_frame=0, channels [None {type=MovieSceneScriptingByteChannel, const [1,1]}]}]
      }]
    }
    DescriptionOLD {
     id=#69
     possessed_class=EditableTextBox
     tracks [
      {
       type=MovieSceneByteTrack
       display_name=Visibility
       sections [{type=MovieSceneByteSection, end_frame=10, start_frame=-1, channels [None {type=MovieSceneScriptingByteChannel, const [1,1]}]}]
      }
      {
       type=MovieScene2DTransformTrack
       sections [{
         type=MovieScene2DTransformSection
         end_frame=10
         start_frame=-1
         channels [
          Translation.X {type=MovieSceneScriptingFloatChannel, keys [[-0.0013,0],[4.4523,130],[10,185]]}
          Translation.Y {type=MovieSceneScriptingFloatChannel, const [0,2]}
          Angle {type=MovieSceneScriptingFloatChannel, const [0,1]}
          Scale.X {type=MovieSceneScriptingFloatChannel, const [1,1]}
          Scale.Y {type=MovieSceneScriptingFloatChannel, const [1,1]}
          Shear.X {type=MovieSceneScriptingFloatChannel, const [0,1]}
          Shear.Y {type=MovieSceneScriptingFloatChannel, const [0,1]}
         ]
        }]
      }
      {
       type=MovieSceneFloatTrack
       display_name=Render Opacity
       sections [{
         type=MovieSceneFloatSection
         end_frame=10
         start_frame=0
         channels [None {type=MovieSceneScriptingFloatChannel, keys [[0,1],[4.4523,1],[10,0]]}]
        }]
      }
     ]
    }
    Image_SmallIndicator_Child {
     id=#44
     possessed_class=Image
     tracks [{
       type=MovieSceneByteTrack
       display_name=Visibility
       sections [{type=MovieSceneByteSection, end_frame=9, start_frame=0, channels [None {type=MovieSceneScriptingByteChannel, const [1,1]}]}]
      }]
    }
    Sign {
     id=#29
     possessed_class=Image
     tracks [
      {
       type=MovieScene2DTransformTrack
       sections [{
         type=MovieScene2DTransformSection
         end_frame=10
         start_frame=-1
         channels [
          Scale.X {type=MovieSceneScriptingFloatChannel, keys [[0.0003,1],[3,0.5],[5.4607,2]]}
          Scale.Y {type=MovieSceneScriptingFloatChannel, keys [[0.0003,1],[3,1.5],[5.4607,0]]}
         ]
        }]
      }
      {
       type=MovieSceneFloatTrack
       display_name=Render Opacity
       sections [{type=MovieSceneFloatSection, end_frame=10, start_frame=0, channels [None {type=MovieSceneScriptingFloatChannel, const [1,1]}]}]
      }
     ]
    }
    Circle {
     id=#30
     possessed_class=Image
     tracks [{
       type=MovieSceneByteTrack
       display_name=Visibility
       sections [{type=MovieSceneByteSection, end_frame=9, start_frame=0, channels [None {type=MovieSceneScriptingByteChannel, const [2,1]}]}]
      }]
    }
    Image_SmallIndicator {
     id=#47
     possessed_class=Image
     tracks [
      {
       type=MovieScene2DTransformTrack
       sections [{
         type=MovieScene2DTransformSection
         end_frame=10
         start_frame=-1
         channels [
          Translation.X {type=MovieSceneScriptingFloatChannel}
          Translation.Y {type=MovieSceneScriptingFloatChannel, keys [[0,55],[4,55],[5.6037,70.4844],[10,85]]}
          Angle {type=MovieSceneScriptingFloatChannel}
          Scale.X {type=MovieSceneScriptingFloatChannel, keys [[0,0.5],[4,0.5],[5.6037,0.7],[10,0.85]]}
          Scale.Y {type=MovieSceneScriptingFloatChannel, keys [[0,0.5],[4,0.5],[5.6037,0.7],[10,0.85]]}
          Shear.X {type=MovieSceneScriptingFloatChannel}
          Shear.Y {type=MovieSceneScriptingFloatChannel}
         ]
        }]
      }
      {
       type=MovieSceneFloatTrack
       display_name=Render Opacity
       sections [{type=MovieSceneFloatSection, end_frame=10, start_frame=0, channels [None {type=MovieSceneScriptingFloatChannel, keys [[4,1],[10,0]]}]}]
      }
      {
       type=MovieSceneByteTrack
       display_name=Visibility
       sections [{type=MovieSceneByteSection, end_frame=10, start_frame=-1, channels [None {type=MovieSceneScriptingByteChannel, const [0,1]}]}]
      }
     ]
    }
   ]}
 }
 ST_Text_Out_Fast {
  binding_count=1
  length_seconds=0.5
  object=&self:ST_Text_Out_Fast
  sequence {bindings [DescriptionOLD {
     id=#79
     display_name=Description
     possessed_class=EditableTextBox
     tracks [
      {
       type=MovieSceneByteTrack
       display_name=Visibility
       sections [{type=MovieSceneByteSection, end_frame=10, start_frame=0, channels [None {type=MovieSceneScriptingByteChannel, const [0,1]}]}]
      }
      {type=MovieScene2DTransformTrack, sections []}
      {
       type=MovieSceneFloatTrack
       display_name=Render Opacity
       sections [{type=MovieSceneFloatSection, end_frame=10, start_frame=0, channels [None {type=MovieSceneScriptingFloatChannel, const [0,1]}]}]
      }
     ]
    }]}
 }
 ST_in_bounce {
  binding_count=13
  length_seconds=0.5956
  object=&self:ST_in_bounce
  sequence {bindings [
    Icon {
     id=#1
     possessed_class=Image
     tracks [
      {
       type=MovieSceneColorTrack
       display_name=Color and Opacity
       sections [{
         type=MovieSceneColorSection
         end_frame=4
         start_frame=0
         channels [
          Color.R {type=MovieSceneScriptingFloatChannel}
          Color.G {type=MovieSceneScriptingFloatChannel}
          Color.B {type=MovieSceneScriptingFloatChannel}
          Color.A {type=MovieSceneScriptingFloatChannel, keys [[1.6527,0],[4.7083,1]]}
         ]
        }]
      }
      {
       type=MovieScene2DTransformTrack
       sections [{
         type=MovieScene2DTransformSection
         end_frame=9
         start_frame=0
         channels [
          Translation.X {type=MovieSceneScriptingFloatChannel, const [0,1]}
          Translation.Y {type=MovieSceneScriptingFloatChannel, const [0,1]}
          Scale.X {type=MovieSceneScriptingFloatChannel, keys [[2.025,1.7],[4.2697,0.8],[5.7663,1.1],[7.637,0.95],[9.508,1]]}
          Scale.Y {type=MovieSceneScriptingFloatChannel, keys [[2.025,0],[4.2697,1.2],[5.7663,0.9],[7.637,1.05],[9.508,1]]}
         ]
        }]
      }
      {
       type=MovieSceneFloatTrack
       display_name=Render Opacity
       sections [{type=MovieSceneFloatSection, end_frame=0, start_frame=0, channels [None {type=MovieSceneScriptingFloatChannel, const [1,1]}]}]
      }
      {
       type=MovieSceneByteTrack
       display_name=Visibility
       sections [{type=MovieSceneByteSection, end_frame=1, start_frame=0, channels [None {type=MovieSceneScriptingByteChannel, const [4,1]}]}]
      }
     ]
    }
    Overlay_0 {
     id=#2
     possessed_class=Overlay
     tracks [{
       type=MovieSceneByteTrack
       display_name=Visibility
       sections [{type=MovieSceneByteSection, end_frame=0, start_frame=0, channels [None {type=MovieSceneScriptingByteChannel, const [4,1]}]}]
      }]
    }
    Circle {
     id=#3
     possessed_class=Image
     tracks [{
       type=MovieSceneByteTrack
       display_name=Visibility
       sections [{type=MovieSceneByteSection, end_frame=2, start_frame=0, channels [None {type=MovieSceneScriptingByteChannel, const [1,1]}]}]
      }]
    }
    DescriptionOLD {
     id=#4
     possessed_class=EditableTextBox
     tracks [
      {
       type=MovieSceneByteTrack
       display_name=Visibility
       sections [{type=MovieSceneByteSection, end_frame=11, start_frame=0, channels [None {type=MovieSceneScriptingByteChannel, const [1,1]}]}]
      }
      {
       type=MovieSceneFloatTrack
       display_name=Render Opacity
       sections [{type=MovieSceneFloatSection, end_frame=0, start_frame=0, channels [None {type=MovieSceneScriptingFloatChannel, const [1,1]}]}]
      }
     ]
    }
    Image_SmallIndicator_Child {
     id=#5
     possessed_class=Image
     tracks [{
       type=MovieSceneByteTrack
       display_name=Visibility
       sections [{type=MovieSceneByteSection, end_frame=11, start_frame=0, channels [None {type=MovieSceneScriptingByteChannel, const [1,2]}]}]
      }]
    }
    Switcher_IndicatorSize {
     id=#6
     possessed_class=WidgetSwitcher
     tracks [{
       type=MovieSceneByteTrack
       display_name=Visibility
       sections [{type=MovieSceneByteSection, end_frame=0, start_frame=0, channels [None {type=MovieSceneScriptingByteChannel, const [4,1]}]}]
      }]
    }
    White_Box_Child {
     id=#7
     possessed_class=Image
     tracks [{
       type=MovieSceneByteTrack
       display_name=Visibility
       sections [{type=MovieSceneByteSection, end_frame=11, start_frame=0, channels [None {type=MovieSceneScriptingByteChannel, const [2,2]}]}]
      }]
    }
    SizeBox_1 {
     id=#8
     possessed_class=SizeBox
     tracks [{
       type=MovieSceneByteTrack
       display_name=Visibility
       sections [{type=MovieSceneByteSection, end_frame=0, start_frame=0, channels [None {type=MovieSceneScriptingByteChannel, const [4,1]}]}]
      }]
    }
    Sign {
     id=#9
     possessed_class=Image
     tracks [
      {
       type=MovieSceneColorTrack
       display_name=Color and Opacity
       sections [{
         type=MovieSceneColorSection
         end_frame=7
         start_frame=0
         channels [
          Color.R {type=MovieSceneScriptingFloatChannel}
          Color.G {type=MovieSceneScriptingFloatChannel}
          Color.B {type=MovieSceneScriptingFloatChannel}
          Color.A {type=MovieSceneScriptingFloatChannel, keys [[3.6253,0],[7.1077,1]]}
         ]
        }]
      }
      {
       type=MovieScene2DTransformTrack
       sections [{
         type=MovieScene2DTransformSection
         end_frame=11
         start_frame=0
         channels [
          Translation.X {type=MovieSceneScriptingFloatChannel, const [0,1]}
          Translation.Y {type=MovieSceneScriptingFloatChannel, const [0,1]}
          Scale.X {type=MovieSceneScriptingFloatChannel, keys [[4.0767,1.7],[6.4257,0.8],[7.9917,1.1],[9.95,0.95],[11.9077,1]]}
          Scale.Y {type=MovieSceneScriptingFloatChannel, keys [[4.0767,0],[6.4257,1.2],[7.9917,0.9],[9.95,1.05],[11.9077,1]]}
         ]
        }]
      }
      {
       type=MovieSceneFloatTrack
       display_name=Render Opacity
       sections [{type=MovieSceneFloatSection, end_frame=0, start_frame=0, channels [None {type=MovieSceneScriptingFloatChannel, const [1,1]}]}]
      }
      {
       type=MovieSceneByteTrack
       display_name=Visibility
       sections [{type=MovieSceneByteSection, end_frame=0, start_frame=0, channels [None {type=MovieSceneScriptingByteChannel, const [4,1]}]}]
      }
     ]
    }
    Overlay_Icon_Box {
     id=#10
     possessed_class=Overlay
     tracks [{
       type=MovieSceneByteTrack
       display_name=Visibility
       sections [{type=MovieSceneByteSection, end_frame=0, start_frame=0, channels [None {type=MovieSceneScriptingByteChannel, const [4,1]}]}]
      }]
    }
    Image_SmallIndicator {
     id=#11
     possessed_class=Image
     tracks [
      {
       type=MovieSceneColorTrack
       display_name=Color and Opacity
       sections [{
         type=MovieSceneColorSection
         end_frame=2
         start_frame=0
         channels [
          Color.R {type=MovieSceneScriptingFloatChannel}
          Color.G {type=MovieSceneScriptingFloatChannel}
          Color.B {type=MovieSceneScriptingFloatChannel}
          Color.A {type=MovieSceneScriptingFloatChannel, keys [[0,0],[2.1093,1]]}
         ]
        }]
      }
      {
       type=MovieScene2DTransformTrack
       sections [{
         type=MovieScene2DTransformSection
         end_frame=7
         start_frame=0
         channels [
          Translation.X {type=MovieSceneScriptingFloatChannel, const [0,1]}
          Translation.Y {type=MovieSceneScriptingFloatChannel, const [46,1]}
          Scale.X {type=MovieSceneScriptingFloatChannel, keys [[0,1.7],[2.0873,0.8],[3.4807,1.1],[5.222,0.95],[6.9637,1]]}
          Scale.Y {type=MovieSceneScriptingFloatChannel, keys [[0,0],[2.0873,1.2],[3.4807,0.9],[5.222,1.05],[6.9637,1]]}
         ]
        }]
      }
      {
       type=MovieSceneByteTrack
       display_name=Visibility
       sections [{type=MovieSceneByteSection, end_frame=0, start_frame=0, channels [None {type=MovieSceneScriptingByteChannel, const [4,1]}]}]
      }
      {
       type=MovieSceneFloatTrack
       display_name=Render Opacity
       sections [{type=MovieSceneFloatSection, end_frame=0, start_frame=0, channels [None {type=MovieSceneScriptingFloatChannel, const [1,1]}]}]
      }
     ]
    }
    White_Box_Solo {
     id=#12
     possessed_class=Image
     tracks [{
       type=MovieSceneByteTrack
       display_name=Visibility
       sections [{type=MovieSceneByteSection, end_frame=11, start_frame=0, channels [None {type=MovieSceneScriptingByteChannel, const [1,1]}]}]
      }]
    }
    Overlay_Icon {
     id=#13
     possessed_class=Overlay
     tracks [
      {
       type=MovieSceneByteTrack
       display_name=Visibility
       sections [{type=MovieSceneByteSection, end_frame=11, start_frame=0, channels [None {type=MovieSceneScriptingByteChannel, const [4,1]}]}]
      }
      {
       type=MovieSceneFloatTrack
       display_name=Render Opacity
       sections [{type=MovieSceneFloatSection, end_frame=0, start_frame=0, channels [None {type=MovieSceneScriptingFloatChannel, const [1,1]}]}]
      }
      {
       type=MovieScene2DTransformTrack
       sections [{
         type=MovieScene2DTransformSection
         end_frame=0
         start_frame=0
         channels [
          Translation.X {type=MovieSceneScriptingFloatChannel, const [0,1]}
          Translation.Y {type=MovieSceneScriptingFloatChannel, const [0,1]}
          Angle {type=MovieSceneScriptingFloatChannel, const [0,1]}
          Scale.X {type=MovieSceneScriptingFloatChannel, const [1,1]}
          Scale.Y {type=MovieSceneScriptingFloatChannel, const [1,1]}
          Shear.X {type=MovieSceneScriptingFloatChannel, const [0,1]}
          Shear.Y {type=MovieSceneScriptingFloatChannel, const [0,1]}
         ]
        }]
      }
     ]
    }
   ]}
 }
]
graphs [EventGraph {nodes [
   Event Construct()
   > "SetText (Text)"(self=Get Description() -> (Description:Common Text Object Reference))
   > n1 = "Play Animation with Finished event (legacy)"!(
    execute:Exec
    Widget:User Widget Object Reference=Self-Reference()
    InAnimation:Widget Animation Object Reference=Get ST_in() -> (ST_in:Widget Animation Object Reference)
    StartAtTime:float=0
    NumLoopsToPlay:int=1
    PlayMode:EUMGSequencePlayMode Enum
    PlaybackSpeed:float=1
   ) -> (then:Exec, Result:UMGSequence Player Object Reference, Finished:Exec) {is=title K2Node_PlayAnimation}
   PlayAnimation(execute=@n1.Finished, InAnimation=Get ST_Text_in() -> (ST_Text_in:Widget Animation Object Reference))
   > Set bIntroAnimationDone(bIntroAnimationDone:bool=false) -> (Output_Get:bool)
   Event PlayOutAnimation() {is=title K2Node_Event}
   n2 = Get ST_Arrow_Disappear() -> (ST_Arrow_Disappear:Widget Animation Object Reference)
   > n3 = Branch(Condition=IsAnimationPlaying(InAnimation=@n2))
   n4 = "Play Animation with Finished event (legacy)"(execute=@n3.else, Widget=Self-Reference(), InAnimation=@n2, StartAtTime=0, NumLoopsToPlay=1, PlaybackSpeed=1)
   > "Play Animation with Finished event (legacy)"(
    Widget=Self-Reference()
    InAnimation=Get ST_Text_Out() -> (ST_Text_Out:Widget Animation Object Reference)
    StartAtTime=0
    NumLoopsToPlay=1
    PlaybackSpeed=1
   )
   > Set bIntroAnimationDone(bIntroAnimationDone=false)
   SetVisibility(execute=@n4.Finished, InVisibility=Collapsed)
   Event PlayBounceAnimation() {is=title K2Node_Event}
   > Branch(Condition=IsVisible())
   > n5 = Branch(Condition=IsAnimationPlaying(InAnimation=Get ST_in_bounce() -> (ST_in_bounce:Widget Animation Object Reference)))
   PlayAnimation(execute=@n5.else, self=Self-Reference(), InAnimation=Get ST_Arrow_Bounce() -> (ST_Arrow_Bounce:Widget Animation Object Reference))
   > PlayAnimation(self=Self-Reference(), InAnimation=Get ST_Text_Out_Fast() -> (ST_Text_Out_Fast:Widget Animation Object Reference))
   > Delay(Duration=0.45)
   > SwitchIndicatorSize(self:Athena Context Tutorial Indicator Object Reference)
   > n6 = Branch(Condition=Get bIntroAnimationDone() -> (bIntroAnimationDone:bool))
   PlayAnimation(execute=@n6.else, InAnimation=Get ST_Text_in())
   > Set bIntroAnimationDone(bIntroAnimationDone=true)
   n7 = Event SetIndicatorText!() -> (OutputDelegate, then, Text:"Text (by ref)") {is=title K2Node_Event}
   > "SetText (Text)"(self=Get Description(), InText=@n7.Text)
   Event PlayInAnimation() {is=title K2Node_Event}
   > SwitchIndicatorSize()
   > n8 = "Play Animation with Finished event (legacy)"(Widget=Self-Reference(), InAnimation=Get ST_in(), StartAtTime=0, NumLoopsToPlay=1, PlaybackSpeed=1)
   PlayAnimation(execute=@n8.Finished, InAnimation=Get ST_Text_in())
  ]}]
variables [bIntroAnimationDone {type=boolean}]
widgets {
 info {inheritedWidgetCount=6, parentClass=/Script/FortniteUI.AthenaContextTutorialIndicator, rootWidgetClass=&SizeBox, widgetCount=26}
 widgets [
  SizeBox_1 {widgetClassPath=&SizeBox, widget=SizeBox_1, properties {bOverride_WidthOverride=true, heightOverride=300, widthOverride=64}}
  Overlay_0 {
   widgetClassPath=&Overlay
   parent=SizeBox_1
   widget=Overlay_0
   properties {renderTransform {angle=0, scale=<1.3 1.3>, shear=<0 0>, translation=<0 0>}}
   slot {class=SizeBoxSlot}
  }
  Overlay_Box_Solo {
   widgetClassPath=&Overlay
   parent=Overlay_0
   widget=Overlay_Box_Solo
   slot {class=OverlaySlot, properties {horizontalAlignment=HAlign_Center, verticalAlignment=VAlign_Center}}
  }
  White_Box_Solo {
   widgetClassPath=&Image
   bIsVariable=true
   parent=Overlay_Box_Solo
   widget=White_Box_Solo
   properties {
    visibility=Collapsed
    brush {
     bIsDynamicallyLoaded=false
     drawAs=Image
     imageType=NoImage
     mirroring=NoMirror
     resourceName=None
     resourceObject=&/MI_Widget_Contextual_IndicatorBox
     tiling=NoTile
     imageSize=<64 64>
     margin=<ltrb 0 0 0 0>
     outlineSettings {
      bUseBrushTransparency=false
      roundingType=HalfHeightRadius
      width=0
      color {colorUseRule=UseColor_Specified, specifiedColor=<rgba 0 0 0 0>}
      cornerRadii=<0 0 0 0>
     }
     tintColor {colorUseRule=UseColor_Specified, specifiedColor=<rgba 1 1 1 1>}
     uVRegion {bIsValid=false, max=<0 0>, min=<0 0>}
    }
   }
   slot {class=OverlaySlot, properties {horizontalAlignment=HAlign_Center, verticalAlignment=VAlign_Center}}
  }
  Switcher_IndicatorSize {
   widgetClassPath=&WidgetSwitcher
   bInherited=true
   bIsVariable=true
   parent=Overlay_0
   widget=Switcher_IndicatorSize
   slot {class=OverlaySlot, properties {horizontalAlignment=HAlign_Center, verticalAlignment=VAlign_Center}}
  }
  Overlay_1 {widgetClassPath=&Overlay, parent=Switcher_IndicatorSize, widget=Overlay_1, slot {class=WidgetSwitcherSlot}}
  Overlay_Icon {widgetClassPath=&Overlay, parent=Overlay_1, widget=Overlay_Icon, slot {class=OverlaySlot}}
  Overlay_Icon_Box {
   widgetClassPath=&Overlay
   parent=Overlay_Icon
   widget=Overlay_Icon_Box
   slot {class=OverlaySlot, properties {horizontalAlignment=HAlign_Center, verticalAlignment=VAlign_Center}}
  }
  Icon {
   widgetClassPath=&Image
   bIsVariable=true
   parent=Overlay_Icon_Box
   widget=Icon
   properties {
    visibility=SelfHitTestInvisible
    brush {
     bIsDynamicallyLoaded=false
     drawAs=Image
     imageType=NoImage
     mirroring=NoMirror
     resourceName=None
     resourceObject=&/MI_Widget_Contextual_Indicator
     tiling=NoTile
     imageSize=<64 64>
     margin=<ltrb 0 0 0 0>
     outlineSettings {
      bUseBrushTransparency=false
      roundingType=HalfHeightRadius
      width=0
      color {colorUseRule=UseColor_Specified, specifiedColor=<rgba 0 0 0 0>}
      cornerRadii=<0 0 0 0>
     }
     tintColor {colorUseRule=UseColor_Specified, specifiedColor=<rgba 1 1 1 1>}
     uVRegion {bIsValid=false, max=<0 0>, min=<0 0>}
    }
    colorAndOpacity=<rgba 0 0.8714 1 1>
   }
   slot {class=OverlaySlot, properties {horizontalAlignment=HAlign_Center, verticalAlignment=VAlign_Center}}
  }
  White_Box_Child {
   widgetClassPath=&Image
   bIsVariable=true
   parent=Overlay_Icon_Box
   widget=White_Box_Child
   properties {
    visibility=SelfHitTestInvisible
    brush {
     bIsDynamicallyLoaded=false
     drawAs=Image
     imageType=NoImage
     mirroring=NoMirror
     resourceName=None
     resourceObject=&MI_Widget_Contextual_IndicatorBox
     tiling=NoTile
     imageSize=<64 64>
     margin=<ltrb 0 0 0 0>
     outlineSettings {
      bUseBrushTransparency=false
      roundingType=HalfHeightRadius
      width=0
      color {colorUseRule=UseColor_Specified, specifiedColor=<rgba 0 0 0 0>}
      cornerRadii=<0 0 0 0>
     }
     tintColor {colorUseRule=UseColor_Specified, specifiedColor=<rgba 1 1 1 1>}
     uVRegion {bIsValid=false, max=<0 0>, min=<0 0>}
    }
   }
   slot {class=OverlaySlot, properties {horizontalAlignment=HAlign_Center, verticalAlignment=VAlign_Center}}
  }
  Sign {
   widgetClassPath=&Image
   bIsVariable=true
   parent=Overlay_Icon
   widget=Sign
   properties {
    visibility=SelfHitTestInvisible
    brush {
     bIsDynamicallyLoaded=false
     drawAs=Image
     imageType=NoImage
     mirroring=NoMirror
     resourceName=None
     resourceObject=&/MI_Widget_Contextual_Sign
     tiling=NoTile
     imageSize=<64 64>
     margin=<ltrb 0 0 0 0>
     outlineSettings {
      bUseBrushTransparency=false
      roundingType=HalfHeightRadius
      width=0
      color {colorUseRule=UseColor_Specified, specifiedColor=<rgba 0 0 0 0>}
      cornerRadii=<0 0 0 0>
     }
     tintColor {colorUseRule=UseColor_Specified, specifiedColor=<rgba 1 1 1 1>}
     uVRegion {bIsValid=false, max=<0 0>, min=<0 0>}
    }
    renderTransformPivot=<0.5 0.8>
   }
   slot {class=OverlaySlot, properties {horizontalAlignment=HAlign_Center, verticalAlignment=VAlign_Center, padding=<ltrb 3 -12 0 0>}}
  }
  Circle {
   widgetClassPath=&Image
   bIsVariable=true
   parent=Overlay_Icon
   widget=Circle
   properties {
    visibility=Hidden
    brush {
     bIsDynamicallyLoaded=false
     drawAs=Image
     imageType=NoImage
     mirroring=NoMirror
     resourceName=None
     resourceObject=&/MI_Widget_Contextual_Circle
     tiling=NoTile
     imageSize=<32 32>
     margin=<ltrb 0 0 0 0>
     outlineSettings {
      bUseBrushTransparency=false
      roundingType=HalfHeightRadius
      width=0
      color {colorUseRule=UseColor_Specified, specifiedColor=<rgba 0 0 0 0>}
      cornerRadii=<0 0 0 0>
     }
     tintColor {colorUseRule=UseColor_Specified, specifiedColor=<rgba 1 1 1 1>}
     uVRegion {bIsValid=false, max=<0 0>, min=<0 0>}
    }
   }
   slot {class=OverlaySlot, properties {horizontalAlignment=HAlign_Center, verticalAlignment=VAlign_Center}}
  }
  Description {
   widgetClassPath=&CommonTextBlock
   bIsVariable=true
   parent=Overlay_Icon
   widget=Description
   properties {
    text=Text Block
    visibility=Collapsed
    font {
     bForceMonospaced=false
     bMaterialIsStencil=false
     fontMaterial=null
     fontObject=&BurbankSmall
     letterSpacing=0
     monospacedWidth=1
     size=22
     skewAmount=0
     typefaceFontName=Bold
     outlineSettings {
      bApplyOutlineToDropShadows=false
      bMiteredCorners=false
      bSeparateFillAlpha=false
      outlineMaterial=null
      outlineSize=0
      outlineColor=<rgba 0 0 0 1>
     }
    }
    renderTransform {angle=0, scale=<1 1>, shear=<0 0>, translation=<70 0>}
    shadowOffset=<0 0>
   }
   slot {class=OverlaySlot, properties {verticalAlignment=VAlign_Center}}
  }
  DescriptionOLD {
   widgetClassPath=&EditableTextBox
   bIsVariable=true
   parent=Overlay_Icon
   widget=DescriptionOLD
   properties {
    text="Ofbva asbfbfkajsdbj bajksb kjbfbfqkjbfkqqwjkdfqjk kqjkjq "
    visibility=Hidden
    renderTransformPivot=<0 0.5>
    widgetStyle {
     backgroundColor {colorUseRule=UseColor_Specified, specifiedColor=<rgba 0 0 0 0>}
     backgroundImageFocused {
      bIsDynamicallyLoaded=false
      drawAs=Box
      imageType=FullColor
      mirroring=NoMirror
      resourceName=../../../Engine/Content/Slate/Common/TextBox_Hovered.png
      resourceObject=null
      tiling=NoTile
      imageSize=<0 0>
      margin=<ltrb 0.25 0.25 0.25 0.25>
      outlineSettings {
       bUseBrushTransparency=false
       roundingType=HalfHeightRadius
       width=0
       color {colorUseRule=UseColor_Specified, specifiedColor=<rgba 0 0 0 0>}
       cornerRadii=<0 0 0 0>
      }
      tintColor {colorUseRule=UseColor_Specified, specifiedColor=<rgba 1 1 1 1>}
      uVRegion {bIsValid=false, max=<0 0>, min=<0 0>}
     }
     backgroundImageHovered {
      bIsDynamicallyLoaded=false
      drawAs=Box
      imageType=FullColor
      mirroring=NoMirror
      resourceName=../../../Engine/Content/Slate/Common/TextBox_Hovered.png
      resourceObject=null
      tiling=NoTile
      imageSize=<0 0>
      margin=<ltrb 0.25 0.25 0.25 0.25>
      outlineSettings {
       bUseBrushTransparency=false
       roundingType=HalfHeightRadius
       width=0
       color {colorUseRule=UseColor_Specified, specifiedColor=<rgba 0 0 0 0>}
       cornerRadii=<0 0 0 0>
      }
      tintColor {colorUseRule=UseColor_Specified, specifiedColor=<rgba 1 1 1 1>}
      uVRegion {bIsValid=false, max=<0 0>, min=<0 0>}
     }
     backgroundImageNormal {
      bIsDynamicallyLoaded=false
      drawAs=Box
      imageType=FullColor
      mirroring=NoMirror
      resourceName=../../../Engine/Content/Slate/Common/TextBox.png
      resourceObject=null
      tiling=NoTile
      imageSize=<0 0>
      margin=<ltrb 0.25 0.25 0.25 0.25>
      outlineSettings {
       bUseBrushTransparency=false
       roundingType=HalfHeightRadius
       width=0
       color {colorUseRule=UseColor_Specified, specifiedColor=<rgba 0 0 0 0>}
       cornerRadii=<0 0 0 0>
      }
      tintColor {colorUseRule=UseColor_Specified, specifiedColor=<rgba 1 1 1 1>}
      uVRegion {bIsValid=false, max=<0 0>, min=<0 0>}
     }
     backgroundImageReadOnly {
      bIsDynamicallyLoaded=false
      drawAs=Box
      imageType=FullColor
      mirroring=NoMirror
      resourceName=../../../Engine/Content/Slate/Common/TextBox_ReadOnly.png
      resourceObject=null
      tiling=NoTile
      imageSize=<0 0>
      margin=<ltrb 0.25 0.25 0.25 0.25>
      outlineSettings {
       bUseBrushTransparency=false
       roundingType=HalfHeightRadius
       width=0
       color {colorUseRule=UseColor_Specified, specifiedColor=<rgba 0 0 0 0>}
       cornerRadii=<0 0 0 0>
      }
      tintColor {colorUseRule=UseColor_Specified, specifiedColor=<rgba 1 1 1 1>}
      uVRegion {bIsValid=false, max=<0 0>, min=<0 0>}
     }
     focusedForegroundColor {colorUseRule=UseColor_Foreground, specifiedColor=<rgba 1 0 1 1>}
     foregroundColor {colorUseRule=UseColor_Foreground, specifiedColor=<rgba 1 1 1 1>}
     hScrollBarPadding=<ltrb 0 0 0 0>
     padding=<ltrb 0 0 0 0>
     readOnlyForegroundColor {colorUseRule=UseColor_Foreground, specifiedColor=<rgba 1 0 1 1>}
     scrollBarStyle {
      thickness=16
      draggedThumbImage {
       bIsDynamicallyLoaded=false
       drawAs=Box
       imageType=FullColor
       mirroring=NoMirror
       resourceName=../../../Engine/Content/Slate/Common/Scrollbar_Thumb.png
       resourceObject=null
       tiling=NoTile
       imageSize=<0 0>
       margin=<ltrb 0.25 0.25 0.25 0.25>
       outlineSettings {
        bUseBrushTransparency=false
        roundingType=HalfHeightRadius
        width=0
        color {colorUseRule=UseColor_Specified, specifiedColor=<rgba 0 0 0 0>}
        cornerRadii=<0 0 0 0>
       }
       tintColor {colorUseRule=UseColor_Specified, specifiedColor=<rgba 1 1 1 1>}
       uVRegion {bIsValid=false, max=<0 0>, min=<0 0>}
      }
      horizontalBackgroundImage {
       bIsDynamicallyLoaded=false
       drawAs=NoDrawType
       imageType=NoImage
       mirroring=NoMirror
       resourceName=None
       resourceObject=null
       tiling=NoTile
       imageSize=<0 0>
       margin=<ltrb 0 0 0 0>
       outlineSettings {
        bUseBrushTransparency=false
        roundingType=HalfHeightRadius
        width=0
        color {colorUseRule=UseColor_Specified, specifiedColor=<rgba 0 0 0 0>}
        cornerRadii=<0 0 0 0>
       }
       tintColor {colorUseRule=UseColor_Specified, specifiedColor=<rgba 1 1 1 1>}
       uVRegion {bIsValid=false, max=<0 0>, min=<0 0>}
      }
      horizontalBottomSlotImage {
       bIsDynamicallyLoaded=false
       drawAs=Image
       imageType=FullColor
       mirroring=NoMirror
       resourceName=../../../Engine/Content/Slate/Common/Scrollbar_Background_Horizontal.png
       resourceObject=null
       tiling=NoTile
       imageSize=<8 8>
       margin=<ltrb 0 0 0 0>
       outlineSettings {
        bUseBrushTransparency=false
        roundingType=HalfHeightRadius
        width=0
        color {colorUseRule=UseColor_Specified, specifiedColor=<rgba 0 0 0 0>}
        cornerRadii=<0 0 0 0>
       }
       tintColor {colorUseRule=UseColor_Specified, specifiedColor=<rgba 1 1 1 1>}
       uVRegion {bIsValid=false, max=<0 0>, min=<0 0>}
      }
      horizontalTopSlotImage {
       bIsDynamicallyLoaded=false
       drawAs=Image
       imageType=FullColor
       mirroring=NoMirror
       resourceName=../../../Engine/Content/Slate/Common/Scrollbar_Background_Horizontal.png
       resourceObject=null
       tiling=NoTile
       imageSize=<8 8>
       margin=<ltrb 0 0 0 0>
       outlineSettings {
        bUseBrushTransparency=false
        roundingType=HalfHeightRadius
        width=0
        color {colorUseRule=UseColor_Specified, specifiedColor=<rgba 0 0 0 0>}
        cornerRadii=<0 0 0 0>
       }
       tintColor {colorUseRule=UseColor_Specified, specifiedColor=<rgba 1 1 1 1>}
       uVRegion {bIsValid=false, max=<0 0>, min=<0 0>}
      }
      hoveredThumbImage {
       bIsDynamicallyLoaded=false
       drawAs=Box
       imageType=FullColor
       mirroring=NoMirror
       resourceName=../../../Engine/Content/Slate/Common/Scrollbar_Thumb.png
       resourceObject=null
       tiling=NoTile
       imageSize=<0 0>
       margin=<ltrb 0.25 0.25 0.25 0.25>
       outlineSettings {
        bUseBrushTransparency=false
        roundingType=HalfHeightRadius
        width=0
        color {colorUseRule=UseColor_Specified, specifiedColor=<rgba 0 0 0 0>}
        cornerRadii=<0 0 0 0>
       }
       tintColor {colorUseRule=UseColor_Specified, specifiedColor=<rgba 1 1 1 1>}
       uVRegion {bIsValid=false, max=<0 0>, min=<0 0>}
      }
      normalThumbImage {
       bIsDynamicallyLoaded=false
       drawAs=Box
       imageType=FullColor
       mirroring=NoMirror
       resourceName=../../../Engine/Content/Slate/Common/Scrollbar_Thumb.png
       resourceObject=null
       tiling=NoTile
       imageSize=<0 0>
       margin=<ltrb 0.25 0.25 0.25 0.25>
       outlineSettings {
        bUseBrushTransparency=false
        roundingType=HalfHeightRadius
        width=0
        color {colorUseRule=UseColor_Specified, specifiedColor=<rgba 0 0 0 0>}
        cornerRadii=<0 0 0 0>
       }
       tintColor {colorUseRule=UseColor_Specified, specifiedColor=<rgba 1 1 1 1>}
       uVRegion {bIsValid=false, max=<0 0>, min=<0 0>}
      }
      verticalBackgroundImage {
       bIsDynamicallyLoaded=false
       drawAs=NoDrawType
       imageType=NoImage
       mirroring=NoMirror
       resourceName=None
       resourceObject=null
       tiling=NoTile
       imageSize=<0 0>
       margin=<ltrb 0 0 0 0>
       outlineSettings {
        bUseBrushTransparency=false
        roundingType=HalfHeightRadius
        width=0
        color {colorUseRule=UseColor_Specified, specifiedColor=<rgba 0 0 0 0>}
        cornerRadii=<0 0 0 0>
       }
       tintColor {colorUseRule=UseColor_Specified, specifiedColor=<rgba 1 1 1 1>}
       uVRegion {bIsValid=false, max=<0 0>, min=<0 0>}
      }
      verticalBottomSlotImage {
       bIsDynamicallyLoaded=false
       drawAs=Image
       imageType=FullColor
       mirroring=NoMirror
       resourceName=../../../Engine/Content/Slate/Common/Scrollbar_Background_Vertical.png
       resourceObject=null
       tiling=NoTile
       imageSize=<8 8>
       margin=<ltrb 0 0 0 0>
       outlineSettings {
        bUseBrushTransparency=false
        roundingType=HalfHeightRadius
        width=0
        color {colorUseRule=UseColor_Specified, specifiedColor=<rgba 0 0 0 0>}
        cornerRadii=<0 0 0 0>
       }
       tintColor {colorUseRule=UseColor_Specified, specifiedColor=<rgba 1 1 1 1>}
       uVRegion {bIsValid=false, max=<0 0>, min=<0 0>}
      }
      verticalTopSlotImage {
       bIsDynamicallyLoaded=false
       drawAs=Image
       imageType=FullColor
       mirroring=NoMirror
       resourceName=../../../Engine/Content/Slate/Common/Scrollbar_Background_Vertical.png
       resourceObject=null
       tiling=NoTile
       imageSize=<8 8>
       margin=<ltrb 0 0 0 0>
       outlineSettings {
        bUseBrushTransparency=false
        roundingType=HalfHeightRadius
        width=0
        color {colorUseRule=UseColor_Specified, specifiedColor=<rgba 0 0 0 0>}
        cornerRadii=<0 0 0 0>
       }
       tintColor {colorUseRule=UseColor_Specified, specifiedColor=<rgba 1 1 1 1>}
       uVRegion {bIsValid=false, max=<0 0>, min=<0 0>}
      }
     }
     textStyle {
      overflowPolicy=Clip
      transformPolicy=None
      colorAndOpacity {colorUseRule=UseColor_Foreground, specifiedColor=<rgba 1 0 1 1>}
      font {
       bForceMonospaced=false
       bMaterialIsStencil=false
       fontMaterial=null
       fontObject=&BurbankSmall
       letterSpacing=0
       monospacedWidth=1
       size=22
       skewAmount=0
       typefaceFontName=Bold
       outlineSettings {
        bApplyOutlineToDropShadows=false
        bMiteredCorners=false
        bSeparateFillAlpha=false
        outlineMaterial=null
        outlineSize=0
        outlineColor=<rgba 0 0 0 1>
       }
      }
      highlightColor {colorUseRule=UseColor_Specified, specifiedColor=<rgba 0.02 0.3 0 1>}
      highlightShape {
       bIsDynamicallyLoaded=false
       drawAs=Box
       imageType=FullColor
       mirroring=NoMirror
       resourceName=../../../Engine/Content/Slate/Common/TextBlockHighlightShape.png
       resourceObject=null
       tiling=NoTile
       imageSize=<0 0>
       margin=<ltrb 0.375 0.375 0.375 0.375>
       outlineSettings {
        bUseBrushTransparency=false
        roundingType=HalfHeightRadius
        width=0
        color {colorUseRule=UseColor_Specified, specifiedColor=<rgba 0 0 0 0>}
        cornerRadii=<0 0 0 0>
       }
       tintColor {colorUseRule=UseColor_Specified, specifiedColor=<rgba 1 1 1 1>}
       uVRegion {bIsValid=false, max=<0 0>, min=<0 0>}
      }
      selectedBackgroundColor {colorUseRule=UseColor_Foreground, specifiedColor=<rgba 1 0 1 1>}
      shadowColorAndOpacity=<rgba 0 0 0 1>
      shadowOffset=<0 0>
      strikeBrush {
       bIsDynamicallyLoaded=false
       drawAs=Image
       imageType=NoImage
       mirroring=NoMirror
       resourceName=None
       resourceObject=null
       tiling=NoTile
       imageSize=<32 32>
       margin=<ltrb 0 0 0 0>
       outlineSettings {
        bUseBrushTransparency=false
        roundingType=HalfHeightRadius
        width=0
        color {colorUseRule=UseColor_Specified, specifiedColor=<rgba 0 0 0 0>}
        cornerRadii=<0 0 0 0>
       }
       tintColor {colorUseRule=UseColor_Specified, specifiedColor=<rgba 1 1 1 1>}
       uVRegion {bIsValid=false, max=<0 0>, min=<0 0>}
      }
      underlineBrush {
       bIsDynamicallyLoaded=false
       drawAs=Image
       imageType=NoImage
       mirroring=NoMirror
       resourceName=None
       resourceObject=null
       tiling=NoTile
       imageSize=<32 32>
       margin=<ltrb 0 0 0 0>
       outlineSettings {
        bUseBrushTransparency=false
        roundingType=HalfHeightRadius
        width=0
        color {colorUseRule=UseColor_Specified, specifiedColor=<rgba 0 0 0 0>}
        cornerRadii=<0 0 0 0>
       }
       tintColor {colorUseRule=UseColor_Specified, specifiedColor=<rgba 1 1 1 1>}
       uVRegion {bIsValid=false, max=<0 0>, min=<0 0>}
      }
     }
     vScrollBarPadding=<ltrb 0 0 0 0>
    }
   }
   slot {class=OverlaySlot, properties {horizontalAlignment=HAlign_Fill, verticalAlignment=VAlign_Fill}}
  }
  ArrowParent {
   widgetClassPath=&Overlay
   parent=Overlay_Icon
   widget=ArrowParent
   slot {class=OverlaySlot, properties {horizontalAlignment=HAlign_Center, verticalAlignment=VAlign_Center}}
  }
  Image_SmallIndicator {
   widgetClassPath=&Image
   bIsVariable=true
   parent=ArrowParent
   widget=Image_SmallIndicator
   properties {
    visibility=SelfHitTestInvisible
    brush {
     bIsDynamicallyLoaded=false
     drawAs=Image
     imageType=NoImage
     mirroring=NoMirror
     resourceName=None
     resourceObject=&/MI_Widget_Contextual_Arrow
     tiling=NoTile
     imageSize=<64 64>
     margin=<ltrb 0 0 0 0>
     outlineSettings {
      bUseBrushTransparency=false
      roundingType=HalfHeightRadius
      width=0
      color {colorUseRule=UseColor_Specified, specifiedColor=<rgba 0 0 0 0>}
      cornerRadii=<0 0 0 0>
     }
     tintColor {colorUseRule=UseColor_Specified, specifiedColor=<rgba 1 1 1 1>}
     uVRegion {bIsValid=false, max=<0 0>, min=<0 0>}
    }
    renderTransform {angle=0, scale=<1 1>, shear=<0 0>, translation=<0 50>}
    renderTransformPivot=<0.5 0.3146>
   }
   slot {class=OverlaySlot, properties {horizontalAlignment=HAlign_Center, verticalAlignment=VAlign_Center, padding=<ltrb 0 -32 0 0>}}
  }
  Image_SmallIndicator_Child {
   widgetClassPath=&Image
   bIsVariable=true
   parent=ArrowParent
   widget=Image_SmallIndicator_Child
   properties {
    visibility=Collapsed
    brush {
     bIsDynamicallyLoaded=false
     drawAs=Image
     imageType=NoImage
     mirroring=NoMirror
     resourceName=None
     resourceObject=&MI_Widget_Contextual_Arrow
     tiling=NoTile
     imageSize=<64 64>
     margin=<ltrb 0 0 0 0>
     outlineSettings {
      bUseBrushTransparency=false
      roundingType=HalfHeightRadius
      width=0
      color {colorUseRule=UseColor_Specified, specifiedColor=<rgba 0 0 0 0>}
      cornerRadii=<0 0 0 0>
     }
     tintColor {colorUseRule=UseColor_Specified, specifiedColor=<rgba 1 1 1 1>}
     uVRegion {bIsValid=false, max=<0 0>, min=<0 0>}
    }
    renderTransform {angle=0, scale=<1 1>, shear=<0 0>, translation=<0 50>}
    renderTransformPivot=<0.5 0.3146>
   }
   slot {class=OverlaySlot, properties {horizontalAlignment=HAlign_Center, verticalAlignment=VAlign_Center, padding=<ltrb 0 -32 0 0>}}
  }
  Overlay_2 {widgetClassPath=&Overlay, parent=Switcher_IndicatorSize, widget=Overlay_2, slot {class=WidgetSwitcherSlot}}
  Overlay_Icon_1 {widgetClassPath=&Overlay, bInherited=true, parent=Overlay_2, widget=Overlay_Icon_1, slot {class=OverlaySlot}}
  Icon_1 {
   widgetClassPath=&Image
   bIsVariable=true
   parent=Overlay_Icon_1
   widget=Icon_1
   properties {
    visibility=SelfHitTestInvisible
    brush {
     bIsDynamicallyLoaded=false
     drawAs=Image
     imageType=NoImage
     mirroring=NoMirror
     resourceName=None
     resourceObject=&MI_Widget_Contextual_Indicator
     tiling=NoTile
     imageSize=<64 64>
     margin=<ltrb 0 0 0 0>
     outlineSettings {
      bUseBrushTransparency=false
      roundingType=HalfHeightRadius
      width=0
      color {colorUseRule=UseColor_Specified, specifiedColor=<rgba 0 0 0 0>}
      cornerRadii=<0 0 0 0>
     }
     tintColor {colorUseRule=UseColor_Specified, specifiedColor=<rgba 1 1 1 1>}
     uVRegion {bIsValid=false, max=<0 0>, min=<0 0>}
    }
    colorAndOpacity=<rgba 0 0.8714 1 1>
   }
   slot {class=OverlaySlot, properties {horizontalAlignment=HAlign_Center, verticalAlignment=VAlign_Center}}
  }
  Sign_1 {
   widgetClassPath=&Image
   bIsVariable=true
   parent=Overlay_Icon_1
   widget=Sign_1
   properties {
    visibility=SelfHitTestInvisible
    brush {
     bIsDynamicallyLoaded=false
     drawAs=Image
     imageType=NoImage
     mirroring=NoMirror
     resourceName=None
     resourceObject=&MI_Widget_Contextual_Sign
     tiling=NoTile
     imageSize=<64 64>
     margin=<ltrb 0 0 0 0>
     outlineSettings {
      bUseBrushTransparency=false
      roundingType=HalfHeightRadius
      width=0
      color {colorUseRule=UseColor_Specified, specifiedColor=<rgba 0 0 0 0>}
      cornerRadii=<0 0 0 0>
     }
     tintColor {colorUseRule=UseColor_Specified, specifiedColor=<rgba 1 1 1 1>}
     uVRegion {bIsValid=false, max=<0 0>, min=<0 0>}
    }
    renderTransformPivot=<0.5 0.8>
   }
   slot {class=OverlaySlot, properties {horizontalAlignment=HAlign_Center, verticalAlignment=VAlign_Center, padding=<ltrb 3 -12 0 0>}}
  }
  Circle_1 {
   widgetClassPath=&Image
   bIsVariable=true
   parent=Overlay_Icon_1
   widget=Circle_1
   properties {
    visibility=Hidden
    brush {
     bIsDynamicallyLoaded=false
     drawAs=Image
     imageType=NoImage
     mirroring=NoMirror
     resourceName=None
     resourceObject=&MI_Widget_Contextual_Circle
     tiling=NoTile
     imageSize=<32 32>
     margin=<ltrb 0 0 0 0>
     outlineSettings {
      bUseBrushTransparency=false
      roundingType=HalfHeightRadius
      width=0
      color {colorUseRule=UseColor_Specified, specifiedColor=<rgba 0 0 0 0>}
      cornerRadii=<0 0 0 0>
     }
     tintColor {colorUseRule=UseColor_Specified, specifiedColor=<rgba 1 1 1 1>}
     uVRegion {bIsValid=false, max=<0 0>, min=<0 0>}
    }
   }
   slot {class=OverlaySlot, properties {horizontalAlignment=HAlign_Center, verticalAlignment=VAlign_Center}}
  }
  Panel_SmallIndicator {widgetClassPath=/Script/UMG.PanelWidget, bInherited=true, widget=null}
  Panel_LargeIndicator {widgetClassPath=&PanelWidget, bInherited=true, widget=null}
  Text_Distance {widgetClassPath=&CommonTextBlock, bInherited=true, widget=null}
  TutorialHighlight {widgetClassPath=&TutorialHighlight}
 ]
}
