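# M_Widget_Contextual is a translucent UI material, named and placed for contextual tutorial widgets in Athena.
# It samples an Image texture parameter, defaulting to a packed tutorial marker texture, and picks channels with a static Component Mask parameter, red by default.
# The masked value is multiplied by the RGB of a Color parameter, white by default, allowing packed shapes to be selected and tinted.
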Material /Game/Athena/Tutorial/Contextual/M_Widget_Contextual
graph {
 properties {blendMode=BLEND_Translucent, materialDomain=MD_UI, pixelDepthOffsetMode=PDOM_Legacy, usedShadingModels=Default Lit}
 nodes [
  n1 = TextureSampleParameter2D() {parameterName=Image, samplerType=SAMPLERTYPE_LinearColor, texture=&/T_ContextTutorialMarker_V02_Packed}
  Multiply(
   A=StaticComponentMaskParameter(None=@n1.RGBA) {defaultR=true, parameterName=Component Mask}
   B=VectorParameter() {parameterName=Color, defaultValue=<rgba 1 1 1 1>}.RGB
  )
 ]
}
