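# HUD indicator ball widget for contextual tutorials, showing a sign icon with a directional arrow.
# SetArrowOrientation reveals and rotates the arrow's parent overlay, while Tick restarts the sliding arrow fade whenever it is allowed and not already playing.
# PlayIntro pairs a spinning, shrinking diamond flash with a squash-and-stretch marker reaction, then clears the intro and arrow-animation flags.
# BounceAnimEndCanPlay is set on Construct but never read here.
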WidgetBlueprint /Game/Athena/Tutorial/Contextual/ContextTutorialBall
animations [
 ST_Marker_MoveReaction {
  binding_count=5
  length_seconds=0.5
  object=&self:ST_Marker_MoveReaction
  sequence {bindings [
    Image_SmallIndicator_Child {
     id=#1
     possessed_class=Image
     tracks [{
       type=MovieSceneByteTrack
       display_name=Visibility
       sections [{type=MovieSceneByteSection, end_frame=10, start_frame=0, channels [None {type=MovieSceneScriptingByteChannel, const [1,1]}]}]
      }]
    }
    Icon_1 {
     id=#2
     possessed_class=Image
     tracks [{
       type=MovieScene2DTransformTrack
       sections [{
         type=MovieScene2DTransformSection
         end_frame=10
         start_frame=0
         channels [
          Translation.X {type=MovieSceneScriptingFloatChannel}
          Translation.Y {type=MovieSceneScriptingFloatChannel}
          Angle {type=MovieSceneScriptingFloatChannel}
          Scale.X {type=MovieSceneScriptingFloatChannel, keys [[0,1],[2.2497,0.6],[3.824,1.3],[4.7273,1]]}
          Scale.Y {type=MovieSceneScriptingFloatChannel, keys [[0,1],[2.2497,1.4],[3.824,0.7],[4.7273,1]]}
          Shear.X {type=MovieSceneScriptingFloatChannel}
          Shear.Y {type=MovieSceneScriptingFloatChannel}
         ]
        }]
      }]
    }
    Image_SmallIndicator {
     id=#3
     possessed_class=Image
     tracks [{
       type=MovieSceneByteTrack
       display_name=Visibility
       sections [{type=MovieSceneByteSection, end_frame=10, start_frame=0, channels [None {type=MovieSceneScriptingByteChannel, const [1,1]}]}]
      }]
    }
    Circle_1 {
     id=#4
     possessed_class=Image
     tracks [{
       type=MovieSceneByteTrack
       display_name=Visibility
       sections [{type=MovieSceneByteSection, end_frame=10, start_frame=0, channels [None {type=MovieSceneScriptingByteChannel, const [1,1]}]}]
      }]
    }
    Sign_1 {
     id=#5
     possessed_class=Image
     tracks [{
       type=MovieScene2DTransformTrack
       sections [{
         type=MovieScene2DTransformSection
         end_frame=10
         start_frame=0
         channels [
          Translation.X {type=MovieSceneScriptingFloatChannel}
          Translation.Y {type=MovieSceneScriptingFloatChannel}
          Angle {type=MovieSceneScriptingFloatChannel}
          Scale.X {type=MovieSceneScriptingFloatChannel, thinned_from=7, keys [[0,1],[2.36,0.4],[3.8247,1],[4.889,1.2],[5.796,1],[6.7497,1]]}
          Scale.Y {type=MovieSceneScriptingFloatChannel, thinned_from=7, keys [[0,1],[2.36,1.6],[3.8247,1],[4.889,0.85],[5.796,1],[6.7497,1]]}
          Shear.X {type=MovieSceneScriptingFloatChannel}
          Shear.Y {type=MovieSceneScriptingFloatChannel}
         ]
        }]
      }]
    }
   ]}
 }
 ST_Arrow_Fade {
  binding_count=1
  length_seconds=0.45
  object=&self:ST_Arrow_Fade
  sequence {bindings [Image_SmallIndicator_Child {
     id=#6
     possessed_class=Image
     tracks [
      {
       type=MovieScene2DTransformTrack
       sections [{
         type=MovieScene2DTransformSection
         end_frame=9
         start_frame=0
         channels [
          Translation.Y {type=MovieSceneScriptingFloatChannel, keys [[0,65],[9,80]]}
          Scale.X {type=MovieSceneScriptingFloatChannel, const [0.6,1]}
          Scale.Y {type=MovieSceneScriptingFloatChannel, const [0.6,1]}
          Shear.X {type=MovieSceneScriptingFloatChannel, const [0,1]}
          Shear.Y {type=MovieSceneScriptingFloatChannel, const [0,1]}
         ]
        }]
      }
      {
       type=MovieSceneFloatTrack
       display_name=Render Opacity
       sections [{type=MovieSceneFloatSection, end_frame=9, start_frame=0, channels [None {type=MovieSceneScriptingFloatChannel, keys [[0,1],[9,0]]}]}]
      }
      {
       type=MovieSceneByteTrack
       display_name=Visibility
       sections [{type=MovieSceneByteSection, end_frame=0, start_frame=0, channels [None {type=MovieSceneScriptingByteChannel, const [4,1]}]}]
      }
     ]
    }]}
 }
 ST_in_DiamondWhite_Offscreen {
  binding_count=1
  length_seconds=0.35
  object=&self:ST_in_DiamondWhite_Offscreen
  sequence {bindings [Diamond_1 {
     id=#7
     possessed_class=Image
     tracks [
      {
       type=MovieScene2DTransformTrack
       sections [{
         type=MovieScene2DTransformSection
         end_frame=30
         start_frame=23
         channels [
          Angle {type=MovieSceneScriptingFloatChannel, keys [[23,180],[27.2023,0]]}
          Scale.X {type=MovieSceneScriptingFloatChannel, keys [[23,3.5],[27.2027,1]]}
          Scale.Y {type=MovieSceneScriptingFloatChannel, keys [[23,3.5],[27.2027,1]]}
         ]
        }]
      }
      {
       type=MovieSceneFloatTrack
       display_name=Render Opacity
       sections [{
         type=MovieSceneFloatSection
         end_frame=30
         start_frame=23
         channels [None {type=MovieSceneScriptingFloatChannel, keys [[23,0],[25.1013,1],[30,0]]}]
        }]
      }
      {
       type=MovieSceneByteTrack
       display_name=Visibility
       sections [{type=MovieSceneByteSection, end_frame=30, start_frame=23, channels [None {type=MovieSceneScriptingByteChannel, keys [[23,3],[29.9997,1]]}]}]
      }
     ]
    }]}
 }
]
graphs [
 EventGraph {nodes [
   n1 = Event SetArrowOrientation!() -> (OutputDelegate, then, Angle:float) {is=title K2Node_Event}
   n2 = Get ArrowParent() -> (ArrowParent:Overlay Object Reference)
   > SetVisibility(self=@n2, InVisibility=SelfHitTestInvisible)
   > SetRenderTransformAngle!(execute, self:Widget Object Reference=@n2, Angle:float=@n1.Angle) -> (then)
   Event Tick()
   > Branch(Condition=Get bCanArrowAnimationPlay() -> (bCanArrowAnimationPlay:bool))
   n3 = Get ST_Arrow_Fade() -> (ST_Arrow_Fade:Widget Animation Object Reference)
   > n4 = Branch(Condition=IsAnimationPlaying(InAnimation=@n3))
   PlayAnimation(execute=@n4.else, InAnimation=@n3)
   Event PlayIntroAnimation() {is=title K2Node_Event}
   > PlayIntro()
   Event Construct()
   > Set BounceAnimIntroCanPlay(BounceAnimIntroCanPlay:bool=true) -> (Output_Get:bool)
   > Set BounceAnimEndCanPlay(BounceAnimEndCanPlay:bool=true) -> (Output_Get:bool)
  ]}
 PlayIntro {nodes [
   PlayIntro() {is=title K2Node_FunctionEntry}
   n1 = Get ST_in_DiamondWhite_Offscreen() -> (ST_in_DiamondWhite_Offscreen:Widget Animation Object Reference)
   > n2 = Branch(Condition=AND Boolean(A=IsAnimationPlaying(InAnimation=@n1), B=Get BounceAnimIntroCanPlay() -> (BounceAnimIntroCanPlay:bool)))
   PlayAnimation(execute=@n2.else, InAnimation=@n1)
   > PlayAnimation(InAnimation=Get ST_Marker_MoveReaction() -> (ST_Marker_MoveReaction:Widget Animation Object Reference))
   > Set BounceAnimIntroCanPlay(BounceAnimIntroCanPlay=false)
   > Set bCanArrowAnimationPlay(bCanArrowAnimationPlay:bool=false) -> (Output_Get:bool)
  ]}
]
variables [
 BounceAnimEndCanPlay {type=boolean}
 BounceAnimIntroCanPlay {type=boolean}
]
widgets {
 info {
  inheritedWidgetCount=1
  parentClass=/Script/FortniteUI.FortHudContextTutorialIndicatorBall
  rootWidgetClass=&SizeBox
  widgetCount=10
 }
 widgets [
  SizeBox {widgetClassPath=&SizeBox, widget=SizeBox, properties {heightOverride=128, widthOverride=128}}
  Overlay_0 {
   widgetClassPath=&Overlay
   parent=SizeBox
   widget=Overlay_0
   properties {renderTransform {angle=0, scale=<1.5 1.5>, shear=<0 0>, translation=<0 0>}}
   slot {class=SizeBoxSlot, properties {horizontalAlignment=HAlign_Center, verticalAlignment=VAlign_Center}}
  }
  Overlay_Icon_1 {
   widgetClassPath=&Overlay
   parent=Overlay_0
   widget=Overlay_Icon_1
   slot {class=OverlaySlot, properties {horizontalAlignment=HAlign_Fill, verticalAlignment=VAlign_Fill}}
  }
  Icon_1 {
   widgetClassPath=&Image
   bIsVariable=true
   parent=Overlay_Icon_1
   widget=Icon_1
   properties {
    visibility=SelfHitTestInvisible
    brush {
     bIsDynamicallyLoaded=false
     drawAs=Image
     imageType=NoImage
     mirroring=NoMirror
     resourceName=None
     resourceObject=&/MI_Widget_Contextual_Indicator
     tiling=NoTile
     imageSize=<64 64>
     margin=<ltrb 0 0 0 0>
     outlineSettings {
      bUseBrushTransparency=false
      roundingType=HalfHeightRadius
      width=0
      color {colorUseRule=UseColor_Specified, specifiedColor=<rgba 0 0 0 0>}
      cornerRadii=<0 0 0 0>
     }
     tintColor {colorUseRule=UseColor_Specified, specifiedColor=<rgba 1 1 1 1>}
     uVRegion {bIsValid=false, max=<0 0>, min=<0 0>}
    }
    colorAndOpacity=<rgba 0 0.8714 1 1>
   }
   slot {class=OverlaySlot, properties {horizontalAlignment=HAlign_Center, verticalAlignment=VAlign_Center}}
  }
  Sign_1 {
   widgetClassPath=&Image
   bIsVariable=true
   parent=Overlay_Icon_1
   widget=Sign_1
   properties {
    visibility=SelfHitTestInvisible
    brush {
     bIsDynamicallyLoaded=false
     drawAs=Image
     imageType=NoImage
     mirroring=NoMirror
     resourceName=None
     resourceObject=&/MI_Widget_Contextual_Sign
     tiling=NoTile
     imageSize=<64 64>
     margin=<ltrb 0 0 0 0>
     outlineSettings {
      bUseBrushTransparency=false
      roundingType=HalfHeightRadius
      width=0
      color {colorUseRule=UseColor_Specified, specifiedColor=<rgba 0 0 0 0>}
      cornerRadii=<0 0 0 0>
     }
     tintColor {colorUseRule=UseColor_Specified, specifiedColor=<rgba 1 1 1 1>}
     uVRegion {bIsValid=false, max=<0 0>, min=<0 0>}
    }
    renderTransformPivot=<0.5 0.8>
   }
   slot {class=OverlaySlot, properties {horizontalAlignment=HAlign_Center, verticalAlignment=VAlign_Center, padding=<ltrb 3 -12 0 0>}}
  }
  Circle_1 {
   widgetClassPath=&Image
   bIsVariable=true
   parent=Overlay_Icon_1
   widget=Circle_1
   properties {
    visibility=Hidden
    brush {
     bIsDynamicallyLoaded=false
     drawAs=Image
     imageType=NoImage
     mirroring=NoMirror
     resourceName=None
     resourceObject=&/MI_Widget_Contextual_Circle
     tiling=NoTile
     imageSize=<32 32>
     margin=<ltrb 0 0 0 0>
     outlineSettings {
      bUseBrushTransparency=false
      roundingType=HalfHeightRadius
      width=0
      color {colorUseRule=UseColor_Specified, specifiedColor=<rgba 0 0 0 0>}
      cornerRadii=<0 0 0 0>
     }
     tintColor {colorUseRule=UseColor_Specified, specifiedColor=<rgba 1 1 1 1>}
     uVRegion {bIsValid=false, max=<0 0>, min=<0 0>}
    }
   }
   slot {class=OverlaySlot, properties {horizontalAlignment=HAlign_Center, verticalAlignment=VAlign_Center}}
  }
  Diamond_1 {
   widgetClassPath=&Image
   bIsVariable=true
   parent=Overlay_Icon_1
   widget=Diamond_1
   properties {
    visibility=Collapsed
    brush {
     bIsDynamicallyLoaded=false
     drawAs=Image
     imageType=NoImage
     mirroring=NoMirror
     resourceName=None
     resourceObject=&/MI_Widget_Contextual_IndicatorBox
     tiling=NoTile
     imageSize=<64 64>
     margin=<ltrb 0 0 0 0>
     outlineSettings {
      bUseBrushTransparency=false
      roundingType=HalfHeightRadius
      width=0
      color {colorUseRule=UseColor_Specified, specifiedColor=<rgba 0 0 0 0>}
      cornerRadii=<0 0 0 0>
     }
     tintColor {colorUseRule=UseColor_Specified, specifiedColor=<rgba 1 1 1 1>}
     uVRegion {bIsValid=false, max=<0 0>, min=<0 0>}
    }
   }
   slot {class=OverlaySlot, properties {horizontalAlignment=HAlign_Center, verticalAlignment=VAlign_Center}}
  }
  ArrowParent {
   widgetClassPath=&Overlay
   bIsVariable=true
   parent=Overlay_0
   widget=ArrowParent
   slot {class=OverlaySlot, properties {horizontalAlignment=HAlign_Center, verticalAlignment=VAlign_Center}}
  }
  Image_SmallIndicator_Child {
   widgetClassPath=&Image
   bIsVariable=true
   parent=ArrowParent
   widget=Image_SmallIndicator_Child
   properties {
    visibility=Collapsed
    brush {
     bIsDynamicallyLoaded=false
     drawAs=Image
     imageType=NoImage
     mirroring=NoMirror
     resourceName=None
     resourceObject=&/MI_Widget_Contextual_Arrow
     tiling=NoTile
     imageSize=<64 64>
     margin=<ltrb 0 0 0 0>
     outlineSettings {
      bUseBrushTransparency=false
      roundingType=HalfHeightRadius
      width=0
      color {colorUseRule=UseColor_Specified, specifiedColor=<rgba 0 0 0 0>}
      cornerRadii=<0 0 0 0>
     }
     tintColor {colorUseRule=UseColor_Specified, specifiedColor=<rgba 1 1 1 1>}
     uVRegion {bIsValid=false, max=<0 0>, min=<0 0>}
    }
    renderTransform {angle=0, scale=<0.6 0.6>, shear=<0 0>, translation=<12 65>}
    renderTransformPivot=<0 0>
   }
   slot {class=OverlaySlot, properties {horizontalAlignment=HAlign_Center, verticalAlignment=VAlign_Center, padding=<ltrb 0 -32 0 0>}}
  }
  TutorialHighlight {widgetClassPath=&TutorialHighlight}
 ]
}
